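# Parent material for the CometHoliday large male character body, built on material attributes with the shading model chosen by expression.
# The shared 2023 character PreFX function feeds two optional static-switch Skin FX layers in series, then the PostFX function.
# The layers reuse weapon-skin functions: Banana, masked by the green channel of PreFX's M output, and SubUVTextures at constant full alpha.
# A flipbook channel mask parameter reads Tech Art Mask but feeds nothing.
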Material /BRCosmetics/Characters/Player/Male/Large/Bodies/M_LRG_CometHoliday/Materials/M_CometHoliday_Parent
graph {
 properties {
  bUseMaterialAttributes=true
  bUsedWithMeshDeformer=true
  bUsedWithMorphTargets=true
  bUsedWithSkeletalMesh=true
  pixelDepthOffsetMode=PDOM_Legacy
  shadingModel=MSM_FromMaterialExpression
  usedShadingModels=Default Lit | Subsurface | Clear Coat | Subsurface Profile | Cloth
 }
 nodes [
  n1 = Constant() {r=1}
  n2 = MaterialFunctionCall() {materialFunction=&MF_CharacterParent_2023_PreFX}
  n3 = StaticSwitchParameter(
   True=MaterialFunctionCall!(Gmap=ComponentMask(None=@n2.M) {g=true}, Material=@n2.Result) {materialFunction=/Game/Weapons/WeaponSkins/MaterialFunction/Banana}.Result
   False=@n2.Result
  ) {group=Skin FX, parameterName=Use Banana}
  MaterialFunctionCall(
   Material=StaticSwitchParameter(
    True=MaterialFunctionCall!(Alpha=@n1, Material=@n3) {materialFunction=/Game/Weapons/WeaponSkins/MaterialFunction/SubUVTextures}.Result
    False=@n3
   ) {group=Skin FX, parameterName=Use Sub UV texture}
   M=@n2.M
   Pre Skinned Local Position=@n2.M
   Distance Blend=@n2.M
   Fresnel=@n2.M
  ) {materialFunction=&MF_CharacterParent_2023_PostFX}
  StaticComponentMaskParameter(None=@n2.Tech Art Mask) {defaultR=true, group=Skin FX, parameterName=Flipbook Texture Channel}
 ]
}
