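# Control Rig for the large male CometHoliday cosmetic body, adding procedural secondary deformation to its shoulders, skirt and strung lights.
# Spherical pose readers, arm-angle remaps and deformRig knee/thigh curves weight additive shoulder and skirt corrective transforms, alongside a shared DeformRig function.
# Weighted multi-bone parent constraints act as surface attachments for light, sphere and aim-target bones, with local-space aim constraints orienting selected dynamic bones.
# A Debug-gated function drives thighs from controls.
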ControlRigBlueprint /BRCosmetics/Characters/Player/Male/Large/Bodies/M_LRG_CometHoliday/Meshes/M_LRG_CometHoliday_CtrlRig
graphs [
 New Function_ContainedGraph {nodes [
   n1 = RigVMFunctionEntryNode()
   RigVMFunction_Sequence = &RigVMFunction_Sequence(ExecuteContext=@n1.ExecuteContext)
   RigVMFunction_Sequence_1 = RigVMAggregateNode(ExecuteContext=@RigVMFunction_Sequence.A)
   &RigUnit_ParentConstraint(
    AdvancedSettings=(RotationOrderForFilter=XZY)
    Child=(Type=Bone,Name="base_spine_lights_01_01_c")
    Parents="((Item=(Type=Bone,Name=\"deform_upperarm_bk_l\"),Weight=0.0186),(Item=(Type=Bone,Name=\"upperarm_twist_01_l\"),Weight=0.0194),(Item=(Type=Bone,Name=\"clavicle_r\"),Weight=0.0206),(Item=(Type=Bone,Name=\"spine_02\"),Weight=0.0232),(Item=(Type=Bone,Name=\"spine_03\"), ...(truncated)"
    Weight=1
    ExecutePin=@RigVMFunction_Sequence_1.A
   )
   > &RigUnit_ParentConstraint(
    AdvancedSettings=(RotationOrderForFilter=XZY)
    Child=(Type=Bone,Name="base_spine_lights_01_02_c")
    Parents="((Item=(Type=Bone,Name=\"spine_01\"),Weight=0.0098),(Item=(Type=Bone,Name=\"spine_02\"),Weight=0.0422),(Item=(Type=Bone,Name=\"clavicle_l\"),Weight=0.0456),(Item=(Type=Bone,Name=\"dyn_clavPush_l\"),Weight=0.0805),(Item=(Type=Bone,Name=\"spine_05\"),Weight=0.1445),(I ...(truncated)"
    Weight=1
   )
   > &RigUnit_ParentConstraint(
    AdvancedSettings=(RotationOrderForFilter=XZY)
    Child=(Type=Bone,Name="base_spine_lights_01_03_c")
    Parents="((Item=(Type=Bone,Name=\"spine_05\"),Weight=0.0335),(Item=(Type=Bone,Name=\"spine_01\"),Weight=0.0396),(Item=(Type=Bone,Name=\"upperarm_r\"),Weight=0.0401),(Item=(Type=Bone,Name=\"dyn_upperArmVolume_r\"),Weight=0.0512),(Item=(Type=Bone,Name=\"upperarm_twist_02_r\"), ...(truncated)"
    Weight=1
   )
   > &RigUnit_ParentConstraint(
    AdvancedSettings=(RotationOrderForFilter=XZY)
    Child=(Type=Bone,Name="base_spine_lights_02_01_c")
    Parents="((Item=(Type=Bone,Name=\"deform_glute_r\"),Weight=0.0017),(Item=(Type=Bone,Name=\"dyn_clavPush_l\"),Weight=0.0047),(Item=(Type=Bone,Name=\"dyn_clavPush_r\"),Weight=0.094),(Item=(Type=Bone,Name=\"pelvis\"),Weight=0.1024),(Item=(Type=Bone,Name=\"spine_01\"),Weight=0.1 ...(truncated)"
    Weight=1
   )
   > &RigUnit_ParentConstraint(
    AdvancedSettings=(RotationOrderForFilter=XZY)
    Child=(Type=Bone,Name="base_spine_lights_02_02_c")
    Parents="((Item=(Type=Bone,Name=\"spine_05\"),Weight=0.0016),(Item=(Type=Bone,Name=\"clavicle_r\"),Weight=0.0101),(Item=(Type=Bone,Name=\"clavicle_l\"),Weight=0.0622),(Item=(Type=Bone,Name=\"pelvis\"),Weight=0.132),(Item=(Type=Bone,Name=\"spine_01\"),Weight=0.1451),(Item=(Ty ...(truncated)"
    Weight=1
   )
   > &RigUnit_ParentConstraint(
    AdvancedSettings=(RotationOrderForFilter=XZY)
    Child=(Type=Bone,Name="base_spine_lights_02_03_c")
    Parents="((Item=(Type=Bone,Name=\"upperarm_twist_01_l\"),Weight=0.0001),(Item=(Type=Bone,Name=\"thigh_twist_01_l\"),Weight=0.0064),(Item=(Type=Bone,Name=\"clavicle_l\"),Weight=0.04),(Item=(Type=Bone,Name=\"spine_04\"),Weight=0.0436),(Item=(Type=Bone,Name=\"pelvis\"),Weight=0 ...(truncated)"
    Weight=1
   )
   > &RigUnit_ParentConstraint(
    AdvancedSettings=(RotationOrderForFilter=XZY)
    Child=(Type=Bone,Name="base_spine_lights_02_04_c")
    Parents="((Item=(Type=Bone,Name=\"dyn_upperArmVolume_r\"),Weight=0.0085),(Item=(Type=Bone,Name=\"upperarm_twist_02_r\"),Weight=0.0216),(Item=(Type=Bone,Name=\"pelvis\"),Weight=0.0433),(Item=(Type=Bone,Name=\"spine_01\"),Weight=0.0626),(Item=(Type=Bone,Name=\"spine_02\"),Weig ...(truncated)"
    Weight=1
   )
   > &RigUnit_ParentConstraint(
    AdvancedSettings=(RotationOrderForFilter=XZY)
    Child=(Type=Bone,Name="base_spine_lights_front_01_01_c")
    Parents="((Item=(Type=Bone,Name=\"deform_pec_r\"),Weight=0.0547),(Item=(Type=Bone,Name=\"dyn_upperArmVolume_r\"),Weight=0.071),(Item=(Type=Bone,Name=\"spine_02\"),Weight=0.0734),(Item=(Type=Bone,Name=\"upperarm_r\"),Weight=0.094),(Item=(Type=Bone,Name=\"upperarm_twist_02_r\" ...(truncated)"
    Weight=1
   )
   > &RigUnit_ParentConstraint(
    AdvancedSettings=(RotationOrderForFilter=XZY)
    Child=(Type=Bone,Name="base_spine_lights_front_01_02_c")
    Parents="((Item=(Type=Bone,Name=\"deform_groin_r\"),Weight=0.0004),(Item=(Type=Bone,Name=\"deform_pec_r\"),Weight=0.0072),(Item=(Type=Bone,Name=\"dyn_clavPush_r\"),Weight=0.0141),(Item=(Type=Bone,Name=\"pelvis\"),Weight=0.0411),(Item=(Type=Bone,Name=\"spine_01\"),Weight=0.07 ...(truncated)"
    Weight=1
   )
   > &RigUnit_ParentConstraint(
    AdvancedSettings=(RotationOrderForFilter=XZY)
    Child=(Type=Bone,Name="base_spine_lights_front_01_03_c")
    Parents="((Item=(Type=Bone,Name=\"thigh_twist_01_r\"),Weight=0.0051),(Item=(Type=Bone,Name=\"deform_groin_l\"),Weight=0.0069),(Item=(Type=Bone,Name=\"thigh_twist_01_l\"),Weight=0.0074),(Item=(Type=Bone,Name=\"spine_04\"),Weight=0.0696),(Item=(Type=Bone,Name=\"spine_01\"),Wei ...(truncated)"
    Weight=1
   )
   > &RigUnit_ParentConstraint(
    AdvancedSettings=(RotationOrderForFilter=XZY)
    Child=(Type=Bone,Name="base_spine_lights_front_01_04_c")
    Parents="((Item=(Type=Bone,Name=\"thigh_twist_01_r\"),Weight=0.002),(Item=(Type=Bone,Name=\"thigh_twist_01_l\"),Weight=0.0279),(Item=(Type=Bone,Name=\"spine_04\"),Weight=0.0368),(Item=(Type=Bone,Name=\"deform_groin_l\"),Weight=0.0443),(Item=(Type=Bone,Name=\"spine_03\"),Weig ...(truncated)"
    Weight=1
   )
   > &RigUnit_ParentConstraint(
    AdvancedSettings=(RotationOrderForFilter=XZY)
    Child=(Type=Bone,Name="base_spine_lights_front_01_05_c")
    Parents="((Item=(Type=Bone,Name=\"clavicle_l\"),Weight=0.0058),(Item=(Type=Bone,Name=\"spine_04\"),Weight=0.0125),(Item=(Type=Bone,Name=\"spine_03\"),Weight=0.0955),(Item=(Type=Bone,Name=\"thigh_twist_01_l\"),Weight=0.1028),(Item=(Type=Bone,Name=\"deform_groin_l\"),Weight=0. ...(truncated)"
    Weight=1
   )
   &RigUnit_ParentConstraint(
    AdvancedSettings=(RotationOrderForFilter=XZY)
    Child=(Type=Bone,Name="base_sphere_01_r")
    Parents="((Item=(Type=Bone,Name=\"spine_01\"),Weight=0.0196),(Item=(Type=Bone,Name=\"spine_04\"),Weight=0.0222),(Item=(Type=Bone,Name=\"deform_upperarm_bk_r\"),Weight=0.0306),(Item=(Type=Bone,Name=\"spine_03\"),Weight=0.1155),(Item=(Type=Bone,Name=\"upperarm_r\"),Weight=0.14 ...(truncated)"
    Weight=1
    ExecutePin=@RigVMFunction_Sequence_1.B
   )
   > &RigUnit_ParentConstraint(
    AdvancedSettings=(RotationOrderForFilter=XZY)
    Child=(Type=Bone,Name="base_arm_rights_01_r")
    Parents="((Item=(Type=Bone,Name=\"deform_upperarm_bk_r\"),Weight=0.0088),(Item=(Type=Bone,Name=\"lowerarm_twist_02_r\"),Weight=0.0126),(Item=(Type=Bone,Name=\"lowerarm_r\"),Weight=0.0134),(Item=(Type=Bone,Name=\"spine_03\"),Weight=0.0139),(Item=(Type=Bone,Name=\"dyn_clavPus ...(truncated)"
    Weight=1
   )
   > &RigUnit_ParentConstraint(
    AdvancedSettings=(RotationOrderForFilter=XZY)
    Child=(Type=Bone,Name="base_arm_rights_02_r")
    Parents="((Item=(Type=Bone,Name=\"spine_05\"),Weight=0.0049),(Item=(Type=Bone,Name=\"deform_elbow_in_r\"),Weight=0.0145),(Item=(Type=Bone,Name=\"lowerarm_twist_02_r\"),Weight=0.023),(Item=(Type=Bone,Name=\"lowerarm_r\"),Weight=0.0337),(Item=(Type=Bone,Name=\"dyn_clavPush_r\" ...(truncated)"
    Weight=1
   )
   > &RigUnit_ParentConstraint(
    AdvancedSettings=(RotationOrderForFilter=XZY)
    Child=(Type=Bone,Name="base_arm_rights_03_r")
    Parents="((Item=(Type=Bone,Name=\"spine_05\"),Weight=0.012),(Item=(Type=Bone,Name=\"spine_04\"),Weight=0.0151),(Item=(Type=Bone,Name=\"spine_01\"),Weight=0.0253),(Item=(Type=Bone,Name=\"spine_03\"),Weight=0.031),(Item=(Type=Bone,Name=\"upperarm_r\"),Weight=0.1709),(Item=(Typ ...(truncated)"
    Weight=1
   )
   > &RigUnit_ParentConstraint(
    AdvancedSettings=(RotationOrderForFilter=XZY)
    Child=(Type=Bone,Name="base_arm_rights_04_r")
    Parents="((Item=(Type=Bone,Name=\"dyn_upperArmVolume_r\"),Weight=0.0016),(Item=(Type=Bone,Name=\"deform_elbow_out_r\"),Weight=0.0037),(Item=(Type=Bone,Name=\"upperarm_twist_02_r\"),Weight=0.0326),(Item=(Type=Bone,Name=\"lowerarm_twist_01_r\"),Weight=0.0355),(Item=(Type=Bon ...(truncated)"
    Weight=1
   )
   > &RigUnit_ParentConstraint(
    AdvancedSettings=(RotationOrderForFilter=XZY)
    Child=(Type=Bone,Name="base_arm_rights_05_r")
    Parents="((Item=(Type=Bone,Name=\"deform_elbow_out_r\"),Weight=0.0002),(Item=(Type=Bone,Name=\"dyn_upperArmVolume_r\"),Weight=0.0012),(Item=(Type=Bone,Name=\"upperarm_twist_02_r\"),Weight=0.0344),(Item=(Type=Bone,Name=\"lowerarm_twist_01_r\"),Weight=0.0877),(Item=(Type=Bon ...(truncated)"
    Weight=1
   )
   > &RigUnit_ParentConstraint(
    AdvancedSettings=(RotationOrderForFilter=XZY)
    Child=(Type=Bone,Name="base_arm_rights_06_r")
    Parents="((Item=(Type=Bone,Name=\"upperarm_twist_02_r\"),Weight=0.0007),(Item=(Type=Bone,Name=\"deform_elbow_out_r\"),Weight=0.1088),(Item=(Type=Bone,Name=\"lowerarm_twist_01_r\"),Weight=0.1293),(Item=(Type=Bone,Name=\"upperarm_r\"),Weight=0.1312),(Item=(Type=Bone,Name=\"lo ...(truncated)"
    Weight=1
   )
   > &RigUnit_ParentConstraint(
    AdvancedSettings=(RotationOrderForFilter=XZY)
    Child=(Type=Bone,Name="base_sphere_02_r")
    Parents="((Item=(Type=Bone,Name=\"dyn_upperArmVolume_r\"),Weight=0.0022),(Item=(Type=Bone,Name=\"upperarm_twist_02_r\"),Weight=0.0366),(Item=(Type=Bone,Name=\"lowerarm_twist_01_r\"),Weight=0.0609),(Item=(Type=Bone,Name=\"lowerarm_r\"),Weight=0.1319),(Item=(Type=Bone,Name=\" ...(truncated)"
    Weight=1
   )
   > &RigUnit_ParentConstraint(
    AdvancedSettings=(RotationOrderForFilter=XZY)
    Child=(Type=Bone,Name="base_arm_lights_01_l")
    Parents="((Item=(Type=Bone,Name=\"upperarm_twist_01_l\"),Weight=0.0021),(Item=(Type=Bone,Name=\"deform_elbow_out_l\"),Weight=0.0104),(Item=(Type=Bone,Name=\"lowerarm_twist_01_l\"),Weight=0.0298),(Item=(Type=Bone,Name=\"upperarm_twist_02_l\"),Weight=0.0335),(Item=(Type=Bone ...(truncated)"
    Weight=1
   )
   > &RigUnit_ParentConstraint(
    AdvancedSettings=(RotationOrderForFilter=XZY)
    Child=(Type=Bone,Name="base_arm_lights_02_l")
    Parents="((Item=(Type=Bone,Name=\"deform_elbow_in_l\"),Weight=0.0004),(Item=(Type=Bone,Name=\"upperarm_twist_01_l\"),Weight=0.0019),(Item=(Type=Bone,Name=\"lowerarm_twist_01_l\"),Weight=0.0118),(Item=(Type=Bone,Name=\"upperarm_twist_02_l\"),Weight=0.0165),(Item=(Type=Bone, ...(truncated)"
    Weight=1
   )
   > &RigUnit_ParentConstraint(
    AdvancedSettings=(RotationOrderForFilter=XZY)
    Child=(Type=Bone,Name="base_arm_lights_03_l")
    Parents="((Item=(Type=Bone,Name=\"deform_elbow_out_l\"),Weight=0.0004),(Item=(Type=Bone,Name=\"lowerarm_twist_01_l\"),Weight=0.0101),(Item=(Type=Bone,Name=\"upperarm_twist_01_l\"),Weight=0.0145),(Item=(Type=Bone,Name=\"lowerarm_l\"),Weight=0.075),(Item=(Type=Bone,Name=\"upp ...(truncated)"
    Weight=1
   )
   > &RigUnit_ParentConstraint(
    AdvancedSettings=(RotationOrderForFilter=XZY)
    Child=(Type=Bone,Name="base_arm_lights_04_l")
    Parents="((Item=(Type=Bone,Name=\"lowerarm_twist_01_l\"),Weight=0.002),(Item=(Type=Bone,Name=\"upperarm_twist_01_l\"),Weight=0.0135),(Item=(Type=Bone,Name=\"upperarm_twist_02_l\"),Weight=0.0836),(Item=(Type=Bone,Name=\"lowerarm_l\"),Weight=0.1027),(Item=(Type=Bone,Name=\"lo ...(truncated)"
    Weight=1
   )
   > &RigUnit_ParentConstraint(
    AdvancedSettings=(RotationOrderForFilter=XZY)
    Child=(Type=Bone,Name="base_arm_lights_06_l")
    Parents="((Item=(Type=Bone,Name=\"spine_04\"),Weight=0.0002),(Item=(Type=Bone,Name=\"lowerarm_twist_01_l\"),Weight=0.0004),(Item=(Type=Bone,Name=\"upperarm_twist_01_l\"),Weight=0.0432),(Item=(Type=Bone,Name=\"lowerarm_l\"),Weight=0.0809),(Item=(Type=Bone,Name=\"deform_elbow ...(truncated)"
    Weight=1
   )
   > &RigUnit_ParentConstraint(
    AdvancedSettings=(RotationOrderForFilter=XZY)
    Child=(Type=Bone,Name="base_arm_lights_07_l")
    Parents="((Item=(Type=Bone,Name=\"spine_04\"),Weight=0.0011),(Item=(Type=Bone,Name=\"lowerarm_twist_01_l\"),Weight=0.0012),(Item=(Type=Bone,Name=\"lowerarm_l\"),Weight=0.0338),(Item=(Type=Bone,Name=\"upperarm_twist_01_l\"),Weight=0.0706),(Item=(Type=Bone,Name=\"lowerarm_twi ...(truncated)"
    Weight=1
   )
   &RigUnit_ParentConstraint(
    AdvancedSettings=(RotationOrderForFilter=XZY)
    Child=(Type=Bone,Name="dyn_Aim_leg_lights_front_02_l")
    Parents="((Item=(Type=Bone,Name=\"skirt_back_01_l\"),Weight=0.0001),(Item=(Type=Bone,Name=\"clavicle_l\"),Weight=0.0002),(Item=(Type=Bone,Name=\"thigh_twist_01_r\"),Weight=0.0002),(Item=(Type=Bone,Name=\"skirt_front_02_c\"),Weight=0.0004),(Item=(Type=Bone,Name=\"spine_04\"), ...(truncated)"
    Weight=1
    ExecutePin=@RigVMFunction_Sequence_1.F
   )
   > &RigUnit_ParentConstraint(
    AdvancedSettings=(RotationOrderForFilter=XZY)
    Child=(Type=Bone,Name="dyn_Aim_leg_lights_front_01_l")
    Parents="((Item=(Type=Bone,Name=\"skirt_side_02_l\"),Weight=0.0129),(Item=(Type=Bone,Name=\"spine_01\"),Weight=0.0283),(Item=(Type=Bone,Name=\"thigh_l\"),Weight=0.0334),(Item=(Type=Bone,Name=\"deform_groin_l\"),Weight=0.0528),(Item=(Type=Bone,Name=\"pelvis\"),Weight=0.082),( ...(truncated)"
    Weight=1
   )
   > &RigUnit_ParentConstraint(
    AdvancedSettings=(RotationOrderForFilter=XZY)
    Child=(Type=Bone,Name="dyn_Aim_sphere_02_r")
    Parents="((Item=(Type=Bone,Name=\"hand_r\"),Weight=0.001),(Item=(Type=Bone,Name=\"upperarm_twist_02_r\"),Weight=0.0023),(Item=(Type=Bone,Name=\"upperarm_r\"),Weight=0.0174),(Item=(Type=Bone,Name=\"deform_elbow_in_r\"),Weight=0.0187),(Item=(Type=Bone,Name=\"lowerarm_r\"),Weig ...(truncated)"
    Weight=1
   )
   > &RigUnit_ParentConstraint(
    AdvancedSettings=(RotationOrderForFilter=XZY)
    Child=(Type=Bone,Name="dyn_Aim_arm_lights_03_l")
    Parents="((Item=(Type=Bone,Name=\"upperarm_twist_02_l\"),Weight=0.0034),(Item=(Type=Bone,Name=\"upperarm_l\"),Weight=0.0236),(Item=(Type=Bone,Name=\"deform_elbow_in_l\"),Weight=0.025),(Item=(Type=Bone,Name=\"lowerarm_l\"),Weight=0.1558),(Item=(Type=Bone,Name=\"lowerarm_twis ...(truncated)"
    Weight=1
   )
   > &RigUnit_ParentConstraint(
    AdvancedSettings=(RotationOrderForFilter=XZY)
    Child=(Type=Bone,Name="dyn_Aim_arm_rights_03_r")
    Parents="((Item=(Type=Bone,Name=\"deform_elbow_out_r\"),Weight=0.0003),(Item=(Type=Bone,Name=\"lowerarm_twist_01_r\"),Weight=0.0131),(Item=(Type=Bone,Name=\"dyn_upperArmVolume_r\"),Weight=0.0147),(Item=(Type=Bone,Name=\"lowerarm_r\"),Weight=0.077),(Item=(Type=Bone,Name=\"up ...(truncated)"
    Weight=1
   )
   > &RigUnit_ParentConstraint(
    AdvancedSettings=(RotationOrderForFilter=XZY)
    Child=(Type=Bone,Name="dyn_Aim_spine_lights_front_01_04_c")
    Parents="((Item=(Type=Bone,Name=\"clavicle_l\"),Weight=0.0002),(Item=(Type=Bone,Name=\"spine_04\"),Weight=0.007),(Item=(Type=Bone,Name=\"skirt_front_02_c\"),Weight=0.0072),(Item=(Type=Bone,Name=\"skirt_side_01_l\"),Weight=0.0372),(Item=(Type=Bone,Name=\"deform_groin_l\"),Wei ...(truncated)"
    Weight=1
   )
   > &RigUnit_ParentConstraint(
    AdvancedSettings=(RotationOrderForFilter=XZY)
    Child=(Type=Bone,Name="dyn_Aim_spine_lights_front_01_03_c")
    Parents="((Item=(Type=Bone,Name=\"skirt_side_01_r\"),Weight=0.0008),(Item=(Type=Bone,Name=\"thigh_twist_01_l\"),Weight=0.0011),(Item=(Type=Bone,Name=\"spine_04\"),Weight=0.0091),(Item=(Type=Bone,Name=\"deform_groin_r\"),Weight=0.0504),(Item=(Type=Bone,Name=\"skirt_front_02_ ...(truncated)"
    Weight=1
   )
   > &RigUnit_ParentConstraint(
    AdvancedSettings=(RotationOrderForFilter=XZY)
    Child=(Type=Bone,Name="dyn_Aim_spine_lights_front_01_02_c")
    Parents="((Item=(Type=Bone,Name=\"dyn_clavPush_r\"),Weight=0.0008),(Item=(Type=Bone,Name=\"skirt_front_02_c\"),Weight=0.0033),(Item=(Type=Bone,Name=\"skirt_side_01_r\"),Weight=0.0054),(Item=(Type=Bone,Name=\"spine_04\"),Weight=0.0195),(Item=(Type=Bone,Name=\"deform_groin_r\" ...(truncated)"
    Weight=1
   )
   &RigUnit_ParentConstraint(
    AdvancedSettings=(RotationOrderForFilter=XZY)
    Child=(Type=Bone,Name="base_leg_lights_front_01_r")
    Parents="((Item=(Type=Bone,Name=\"pelvis\"),Weight=0.0029),(Item=(Type=Bone,Name=\"skirt_side_01_r\"),Weight=0.012),(Item=(Type=Bone,Name=\"skirt_front_03_c\"),Weight=0.0501),(Item=(Type=Bone,Name=\"thigh_twist_01_r\"),Weight=0.056),(Item=(Type=Bone,Name=\"skirt_side_03_r\") ...(truncated)"
    Weight=1
    ExecutePin=@RigVMFunction_Sequence_1.G
   )
   > &RigUnit_ParentConstraint(
    AdvancedSettings=(RotationOrderForFilter=XZY)
    Child=(Type=Bone,Name="base_leg_lights_front_02_r")
    Parents="((Item=(Type=Bone,Name=\"skirt_side_03_r\"),Weight=0.0021),(Item=(Type=Bone,Name=\"pelvis\"),Weight=0.0078),(Item=(Type=Bone,Name=\"skirt_side_01_r\"),Weight=0.0122),(Item=(Type=Bone,Name=\"thigh_twist_01_r\"),Weight=0.0235),(Item=(Type=Bone,Name=\"skirt_side_02_r\" ...(truncated)"
    Weight=1
   )
   > &RigUnit_ParentConstraint(
    AdvancedSettings=(RotationOrderForFilter=XZY)
    Child=(Type=Bone,Name="base_leg_back_lights_02_r")
    Parents="((Item=(Type=Bone,Name=\"spine_01\"),Weight=0.0001),(Item=(Type=Bone,Name=\"deform_glute_r\"),Weight=0.0002),(Item=(Type=Bone,Name=\"thigh_twist_01_r\"),Weight=0.0004),(Item=(Type=Bone,Name=\"pelvis\"),Weight=0.0009),(Item=(Type=Bone,Name=\"skirt_back_01_l\"),Weight ...(truncated)"
    Weight=1
   )
   > &RigUnit_ParentConstraint(
    AdvancedSettings=(RotationOrderForFilter=XZY)
    Child=(Type=Bone,Name="base_leg_lights_front_03_r")
    Parents="((Item=(Type=Bone,Name=\"deform_groin_l\"),Weight=0.0008),(Item=(Type=Bone,Name=\"skirt_side_01_r\"),Weight=0.0013),(Item=(Type=Bone,Name=\"thigh_twist_01_l\"),Weight=0.0024),(Item=(Type=Bone,Name=\"thigh_twist_01_r\"),Weight=0.0113),(Item=(Type=Bone,Name=\"spine_0 ...(truncated)"
    Weight=1
   )
   > &RigUnit_ParentConstraint(
    AdvancedSettings=(RotationOrderForFilter=XZY)
    Child=(Type=Bone,Name="base_leg_back_lights_01_r")
    Parents="((Item=(Type=Bone,Name=\"thigh_twist_01_r\"),Weight=0.0112),(Item=(Type=Bone,Name=\"spine_01\"),Weight=0.0278),(Item=(Type=Bone,Name=\"skirt_back_01_l\"),Weight=0.0298),(Item=(Type=Bone,Name=\"skirt_back_02_l\"),Weight=0.031),(Item=(Type=Bone,Name=\"pelvis\"),Weight ...(truncated)"
    Weight=1
   )
   > &RigUnit_ParentConstraint(
    AdvancedSettings=(RotationOrderForFilter=XZY)
    Child=(Type=Bone,Name="base_leg_back_lights_02_r")
    Parents="((Item=(Type=Bone,Name=\"thigh_twist_01_r\"),Weight=0.0043),(Item=(Type=Bone,Name=\"skirt_side_01_r\"),Weight=0.0054),(Item=(Type=Bone,Name=\"skirt_back_01_l\"),Weight=0.0065),(Item=(Type=Bone,Name=\"pelvis\"),Weight=0.0247),(Item=(Type=Bone,Name=\"skirt_side_02_r\" ...(truncated)"
    Weight=1
   )
   > &RigUnit_ParentConstraint(
    AdvancedSettings=(RotationOrderForFilter=XZY)
    Child=(Type=Bone,Name="base_leg_back_lights_03_r")
    Parents="((Item=(Type=Bone,Name=\"deform_glute_r\"),Weight=0.0004),(Item=(Type=Bone,Name=\"skirt_back_03_r\"),Weight=0.0054),(Item=(Type=Bone,Name=\"pelvis\"),Weight=0.0064),(Item=(Type=Bone,Name=\"thigh_twist_01_r\"),Weight=0.0073),(Item=(Type=Bone,Name=\"skirt_side_01_r\") ...(truncated)"
    Weight=1
   )
   > &RigUnit_ParentConstraint(
    AdvancedSettings=(RotationOrderForFilter=XZY)
    Child=(Type=Bone,Name="base_leg_lights_front_01_l")
    Parents="((Item=(Type=Bone,Name=\"clavicle_l\"),Weight=0.001),(Item=(Type=Bone,Name=\"spine_04\"),Weight=0.0175),(Item=(Type=Bone,Name=\"thigh_twist_01_l\"),Weight=0.0973),(Item=(Type=Bone,Name=\"spine_03\"),Weight=0.1072),(Item=(Type=Bone,Name=\"deform_groin_l\"),Weight=0.1 ...(truncated)"
    Weight=1
   )
   > &RigUnit_ParentConstraint(
    AdvancedSettings=(RotationOrderForFilter=XZY)
    Child=(Type=Bone,Name="base_leg_lights_front_02_l")
    Parents="((Item=(Type=Bone,Name=\"thigh_twist_01_r\"),Weight=0.0001),(Item=(Type=Bone,Name=\"deform_pec_l\"),Weight=0.0001),(Item=(Type=Bone,Name=\"thigh_l\"),Weight=0.0009),(Item=(Type=Bone,Name=\"clavicle_l\"),Weight=0.0047),(Item=(Type=Bone,Name=\"spine_04\"),Weight=0.012 ...(truncated)"
    Weight=1
   )
   > &RigUnit_ParentConstraint(
    AdvancedSettings=(RotationOrderForFilter=XZY)
    Child=(Type=Bone,Name="base_leg_back_lights_01_l")
    Parents="((Item=(Type=Bone,Name=\"spine_03\"),Weight=0.0268),(Item=(Type=Bone,Name=\"spine_02\"),Weight=0.0338),(Item=(Type=Bone,Name=\"thigh_twist_01_l\"),Weight=0.045),(Item=(Type=Bone,Name=\"thigh_l\"),Weight=0.0536),(Item=(Type=Bone,Name=\"skirt_side_01_l\"),Weight=0.075 ...(truncated)"
    Weight=1
   )
   > &RigUnit_ParentConstraint(
    AdvancedSettings=(RotationOrderForFilter=XZY)
    Child=(Type=Bone,Name="base_leg_back_lights_02_l")
    Parents="((Item=(Type=Bone,Name=\"deform_glute_l\"),Weight=0.0213),(Item=(Type=Bone,Name=\"thigh_twist_01_l\"),Weight=0.0285),(Item=(Type=Bone,Name=\"spine_03\"),Weight=0.0291),(Item=(Type=Bone,Name=\"spine_02\"),Weight=0.0387),(Item=(Type=Bone,Name=\"thigh_l\"),Weight=0.046 ...(truncated)"
    Weight=1
   )
   RigVMFunctionReturnNode(ExecuteContext=@RigVMFunction_Sequence.B)
  ]}
 New Function_ContainedGraph {nodes [
   n1 = RigVMFunctionEntryNode()
   RigVMFunction_Sequence = &RigVMFunction_Sequence(ExecuteContext=@n1.ExecuteContext)
   n2 = &RigVMFunction_Sequence(ExecuteContext=@RigVMFunction_Sequence.A)
   n3 = &RigUnit_SphericalPoseReader(
    ActiveRegionScaleFactors=(PositiveWidth=2,NegativeWidth=0.5,PositiveHeight=2,NegativeHeight=1.9)
    ActiveRegionSize=0.2
    Debug=(bDrawDebug=true,bDraw2D=False,bDrawLocalAxes=False,DebugScale=25,DebugSegments=20,DebugThickness=0.25)
    DriverAxis=(X=-1,Y=0,Z=0)
    DriverItem=(Type=Bone,Name="upperarm_r")
    FalloffRegionScaleFactors=(PositiveWidth=1,NegativeWidth=1,PositiveHeight=1,NegativeHeight=1)
    FalloffSize=0.5
    FlipHeightScaling=False
    FlipWidthScaling=False
    OptionalParentItem=(Type=Bone,Name="None")
    RotationOffset=(X=-90,Y=0,Z=0)
    ExecutePin=@n2.A
   )
   > /Script/ControlRig.RigUnit_ModifyBoneTransforms(
    BoneToModify=((Bone="dyn_clavPush_r",Transform=(Rotation=(X=0,Y=0,Z=0,W=1),Translation=(X=-1.5,Y=0,Z=0),Scale3D=(X=1,Y=1.2,Z=1.2))))
    Mode=AdditiveLocal
    WeightMaximum=1
    WeightMinimum=0
    Weight=@n3.OutputParam
   ) {is=struct RigVMUnitNode}
   > &RigUnit_ModifyBoneTransforms(
    BoneToModify=((Bone="dyn_upperArmVolume_r",Transform=(Rotation=(X=0,Y=0,Z=0,W=1),Translation=(X=10,Y=12,Z=6),Scale3D=(X=2,Y=2,Z=2))))
    Mode=AdditiveLocal
    WeightMaximum=1
    WeightMinimum=0
    Weight=@n3.OutputParam
   )
   > &RigUnit_ModifyBoneTransforms(
    BoneToModify=((Bone="upperarm_twist_01_r",Transform=(Rotation=(X=0,Y=0,Z=0,W=1),Translation=(X=0,Y=0,Z=0),Scale3D=(X=1,Y=1,Z=1.5))))
    Mode=AdditiveLocal
    WeightMaximum=1
    WeightMinimum=0
    Weight=@n3.OutputParam
   )
   n4 = &RigUnit_GetRelativeBoneTransform(Bone=upperarm_l, Space=spine_05)
   n5 = &RigVMFunction_MathFloatRemap(
    SourceMaximum=64
    SourceMinimum=150
    TargetMinimum=0
    bClamp=true
    Value=/Script/RigVM.RigVMFunction_MathFloatAbs(Value=&RigVMFunction_MathQuaternionToEuler(RotationOrder=ZYX, Value=@n4.Transform.Rotation).Result.Z) {is=struct RigVMUnitNode, template="Absolute::Execute(in Value,out Result)"}.Result
   )
   > &RigUnit_ModifyBoneTransforms(
    BoneToModify=((Bone="dyn_clavPush_l",Transform=(Rotation=(X=0,Y=0,Z=0,W=1),Translation=(X=6,Y=0,Z=0),Scale3D=(X=1,Y=1,Z=1))))
    Mode=AdditiveLocal
    WeightMaximum=1
    WeightMinimum=0
    Weight=@n5.Result
   )
   > &RigUnit_ModifyBoneTransforms(
    BoneToModify=((Bone="dyn_upperArmVolume_l",Transform=(Rotation=(X=0,Y=0,Z=0,W=1),Translation=(X=-10,Y=-12,Z=-12),Scale3D=(X=2,Y=2,Z=2))))
    Mode=AdditiveLocal
    WeightMaximum=1
    WeightMinimum=0
    Weight=@n5.Result
   )
   RigVMFunctionReturnNode(ExecuteContext=@RigVMFunction_Sequence.B)
   n6 = &RigVMFunction_MathQuaternionToEuler(RotationOrder=ZYX, Value=&RigUnit_GetRelativeBoneTransform(Bone=upperarm_r, Space=spine_05).Transform.Rotation)
   &RigVMFunction_MathFloatRemap(SourceMaximum=122, SourceMinimum=20, TargetMinimum=0, bClamp=true, Value=@n6.Result.Z)
   &RigVMFunction_MathFloatRemap(SourceMaximum=130, SourceMinimum=50, TargetMinimum=0, bClamp=true, Value=@n6.Result.Z)
  ]}
 New Function_ContainedGraph {nodes [
   n1 = RigVMFunctionEntryNode()
   RigVMFunction_ControlFlowBranch = &RigVMFunction_ControlFlowBranch(Condition=@n1.Debug, ExecuteContext=@n1.ExecuteContext)
   RigVMFunctionReturnNode(ExecuteContext=@RigVMFunction_ControlFlowBranch.Completed)
   RigVMFunction_Sequence = RigVMAggregateNode(ExecuteContext=@RigVMFunction_ControlFlowBranch.True)
   RigUnit_SetTransform = &RigUnit_SetTransform(
    Item=(Type=Bone,Name="thigh_r")
    Weight=1
    bInitial=False
    Value=&RigUnit_GetTransform(Item=(Type=Control,Name="thigh_r_ctrl"), bInitial=False).Transform
    ExecutePin=@RigVMFunction_Sequence.A
   )
   RigUnit_GetTransform = &RigUnit_GetTransform(Item=(Type=Control,Name="thigh_l_ctrl"), bInitial=False)
   > &RigUnit_SetTransform(Item=(Type=Bone,Name="thigh_l"), Weight=1, bInitial=False, Value=@RigUnit_GetTransform.Transform)
  ]}
 New Function_ContainedGraph {nodes [
   RigVMFunctionEntryNode()
   > RigVMFunction_Sequence = RigVMAggregateNode()
   n1 = &RigUnit_SphericalPoseReader(
    ActiveRegionScaleFactors=(PositiveWidth=0.5,NegativeWidth=1.5,PositiveHeight=1,NegativeHeight=1.9)
    ActiveRegionSize=0.2
    Debug=(bDrawDebug=false,bDraw2D=false,bDrawLocalAxes=False,DebugScale=25,DebugSegments=20,DebugThickness=0.25)
    DriverAxis=(X=-1,Y=0,Z=0)
    DriverItem=(Type=Bone,Name="thigh_l")
    FalloffRegionScaleFactors=(PositiveWidth=1.3,NegativeWidth=3,PositiveHeight=1,NegativeHeight=3)
    FalloffSize=0.4
    FlipHeightScaling=false
    FlipWidthScaling=false
    OptionalParentItem=(Type=Bone,Name="None")
    RotationOffset=(X=0,Y=-180,Z=0)
    ExecutePin=@RigVMFunction_Sequence.A
   )
   > &RigUnit_ModifyTransforms(
    ItemToModify="((Item=(Type=Bone,Name=\"skirt_side_01_l\"),Transform=(Rotation=(X=0.202,Y=-0.0777,Z=-0.7028,W=0.6776),Translation=(X=-4.4874,Y=1.2024,Z=0),Scale3D=(X=1,Y=1,Z=1))),(Item=(Type=Bone,Name=\"skirt_back_01_l\"),Transform=(Rotation=(X=0,Y=-0.1736,Z=0,W=0.9848),Tran ...(truncated)"
    WeightMinimum=0
    Weight=@n1.OutputParam
   )
   > n2 = &RigUnit_SphericalPoseReader(
    ActiveRegionScaleFactors=(PositiveWidth=0.5,NegativeWidth=1.5,PositiveHeight=1.9,NegativeHeight=1.9)
    ActiveRegionSize=0.2
    Debug=(bDrawDebug=false,bDraw2D=false,bDrawLocalAxes=False,DebugScale=25,DebugSegments=20,DebugThickness=0.25)
    DriverAxis=(X=1,Y=0,Z=0)
    DriverItem=(Type=Bone,Name="thigh_r")
    FalloffRegionScaleFactors=(PositiveWidth=1.3,NegativeWidth=3,PositiveHeight=3,NegativeHeight=3)
    FalloffSize=0.4
    FlipHeightScaling=false
    FlipWidthScaling=false
    OptionalParentItem=(Type=Bone,Name="None")
    RotationOffset=(X=0,Y=0,Z=0)
   )
   > &RigUnit_ModifyTransforms(
    ItemToModify="((Item=(Type=Bone,Name=\"skirt_side_01_r\"),Transform=(Rotation=(X=0.1094,Y=0.2346,Z=-0.4082,W=0.8754),Translation=(X=-4.4874,Y=1.2024,Z=0),Scale3D=(X=1,Y=1,Z=1))),(Item=(Type=Bone,Name=\"skirt_back_01_r\"),Transform=(Rotation=(X=0.0226,Y=0.2578,Z=-0.0842,W=0. ...(truncated)"
    WeightMinimum=0
    Weight=@n2.OutputParam
   )
   n3 = &RigUnit_SphericalPoseReader(
    ActiveRegionScaleFactors=(PositiveWidth=2,NegativeWidth=0.5,PositiveHeight=2,NegativeHeight=2)
    ActiveRegionSize=0.15
    Debug=(bDrawDebug=false,bDraw2D=False,bDrawLocalAxes=False,DebugScale=25,DebugSegments=20,DebugThickness=0.25)
    DriverAxis=(X=-1,Y=0,Z=0)
    DriverItem=(Type=Bone,Name="thigh_l")
    FalloffRegionScaleFactors=(PositiveWidth=1,NegativeWidth=1.7,PositiveHeight=1,NegativeHeight=1)
    FalloffSize=0.35
    FlipHeightScaling=False
    FlipWidthScaling=False
    OptionalParentItem=(Type=Bone,Name="None")
    RotationOffset=(X=-90,Y=0,Z=0)
    ExecutePin=@RigVMFunction_Sequence.B
   )
   > &RigUnit_ModifyTransforms(
    ItemToModify="((Item=(Type=Bone,Name=\"skirt_back_01_l\"),Transform=(Rotation=(X=0,Y=0,Z=-0.8434,W=0.5373),Translation=(X=-2,Y=0,Z=0),Scale3D=(X=1,Y=1,Z=1))),(Item=(Type=Bone,Name=\"base_leg_lights_front_02_l\"),Transform=(Rotation=(X=0,Y=0,Z=0,W=1),Translation=(X=0,Y=0,Z=0 ...(truncated)"
    WeightMinimum=0
    Weight=@n3.OutputParam
   )
   > n4 = &RigUnit_SphericalPoseReader(
    ActiveRegionScaleFactors=(PositiveWidth=2,NegativeWidth=0.5,PositiveHeight=2,NegativeHeight=2)
    ActiveRegionSize=0.1
    Debug=(bDrawDebug=false,bDraw2D=False,bDrawLocalAxes=False,DebugScale=25,DebugSegments=20,DebugThickness=0.25)
    DriverAxis=(X=1,Y=0,Z=0)
    DriverItem=(Type=Bone,Name="thigh_r")
    FalloffRegionScaleFactors=(PositiveWidth=1,NegativeWidth=1.9,PositiveHeight=1,NegativeHeight=1)
    FalloffSize=0.3
    FlipHeightScaling=False
    FlipWidthScaling=False
    OptionalParentItem=(Type=Bone,Name="None")
    RotationOffset=(X=90,Y=180,Z=10)
   )
   > &RigUnit_ModifyTransforms(
    ItemToModify=((Item=(Type=Bone,Name="skirt_back_01_r"),Transform=(Rotation=(X=0,Y=0,Z=-0.866,W=0.5),Translation=(X=-2,Y=0,Z=0),Scale3D=(X=1,Y=1,Z=1))))
    WeightMinimum=0
    Weight=@n4.OutputParam
   )
   &RigUnit_ModifyTransforms(
    ItemToModify="((Item=(Type=Bone,Name=\"skirt_side_01_r\"),Transform=(Rotation=(X=0.0594,Y=0.1632,Z=-0.3368,W=0.9254),Translation=(X=-2,Y=-2,Z=0),Scale3D=(X=1,Y=1,Z=1))),(Item=(Type=Bone,Name=\"base_skirt_back_r\"),Transform=(Rotation=(X=0,Y=0,Z=-0.1736,W=0.9848),Translation ...(truncated)"
    WeightMinimum=0
    Weight=RigVMDispatchNode(BoneName=root, Default=0, Name=deformRig_r_knee_alpha) {template="DISPATCH_RigDispatch_GetAnimAttribute(in Name,in BoneName,in Default,out Value,out Found)"}.Value
    ExecutePin=@RigVMFunction_Sequence.C
   )
   > &RigUnit_ModifyTransforms(
    ItemToModify="((Item=(Type=Bone,Name=\"skirt_side_01_l\"),Transform=(Rotation=(X=-0.0098,Y=-0.2617,Z=-0.1617,W=0.9515),Translation=(X=-2,Y=-2,Z=0),Scale3D=(X=1,Y=1,Z=1))),(Item=(Type=Bone,Name=\"base_skirt_back_l\"),Transform=(Rotation=(X=-0.0038,Y=-0.0435,Z=-0.0871,W=0.995 ...(truncated)"
    WeightMinimum=0
    Weight=RigVMDispatchNode(BoneName=root, Default=0, Name=deformRig_l_knee_alpha) {template="DISPATCH_RigDispatch_GetAnimAttribute(in Name,in BoneName,in Default,out Value,out Found)"}.Value
   )
   n5 = RigVMDispatchNode(BoneName=root, Default=0, Name=deformRig_l_thigh_up_alpha) {template="DISPATCH_RigDispatch_GetAnimAttribute(in Name,in BoneName,in Default,out Value,out Found)"}
   n6 = RigVMDispatchNode(BoneName=root, Default=0, Name=deformRig_l_knee_alpha) {template="DISPATCH_RigDispatch_GetAnimAttribute(in Name,in BoneName,in Default,out Value,out Found)"}
   > &RigUnit_ModifyTransforms(
    ItemToModify=((Item=(Type=Bone,Name="skirt_back_01_l"),Transform=(Rotation=(X=0,Y=0,Z=-0.4226,W=0.9063),Translation=(X=0,Y=0,Z=0),Scale3D=(X=1,Y=1,Z=1))))
    WeightMinimum=0
    Weight=&RigVMFunction_MathDoubleClamp(Value=&RigVMFunction_MathDoubleSub(B=@n5.Value, A=@n6.Value).Result).Result
   )
   n7 = RigVMDispatchNode(BoneName=root, Default=0, Name=deformRig_r_thigh_up_alpha) {template="DISPATCH_RigDispatch_GetAnimAttribute(in Name,in BoneName,in Default,out Value,out Found)"}
   n8 = RigVMDispatchNode(BoneName=root, Default=0, Name=deformRig_r_knee_alpha) {template="DISPATCH_RigDispatch_GetAnimAttribute(in Name,in BoneName,in Default,out Value,out Found)"}
   > &RigUnit_ModifyTransforms(
    ItemToModify=((Item=(Type=Bone,Name="skirt_back_01_r"),Transform=(Rotation=(X=0,Y=0,Z=-0.4226,W=0.9063),Translation=(X=0,Y=0,Z=0),Scale3D=(X=1,Y=1,Z=1))))
    WeightMinimum=0
    Weight=&RigVMFunction_MathDoubleClamp(Value=&RigVMFunction_MathDoubleSub(B=@n7.Value, A=@n8.Value).Result).Result
   )
   n9 = &RigUnit_SphericalPoseReader(
    ActiveRegionScaleFactors=(PositiveWidth=2,NegativeWidth=0.5,PositiveHeight=2,NegativeHeight=2)
    ActiveRegionSize=0.15
    Debug=(bDrawDebug=false,bDraw2D=False,bDrawLocalAxes=False,DebugScale=25,DebugSegments=20,DebugThickness=0.25)
    DriverAxis=(X=-1,Y=0,Z=0)
    DriverItem=(Type=Bone,Name="thigh_l")
    FalloffRegionScaleFactors=(PositiveWidth=2,NegativeWidth=1.8,PositiveHeight=2,NegativeHeight=2)
    FalloffSize=0.35
    FlipHeightScaling=False
    FlipWidthScaling=False
    OptionalParentItem=(Type=Bone,Name="None")
    RotationOffset=(X=90,Y=0,Z=-10)
    ExecutePin=@RigVMFunction_Sequence.D
   )
   > &RigUnit_ModifyTransforms(
    ItemToModify="((Item=(Type=Bone,Name=\"base_skirt_front_c\"),Transform=(Rotation=(X=-0.183,Y=0.183,Z=-0.683,W=0.683),Translation=(X=-22.0006,Y=0,Z=-0.0006),Scale3D=(X=1,Y=1,Z=1))),(Item=(Type=Bone,Name=\"base_skirt_side_l\"),Transform=(Rotation=(X=0,Y=-0.6775,Z=0,W=0.7356), ...(truncated)"
    WeightMinimum=0
    Weight=@n9.OutputParam
   )
   > n10 = &RigUnit_SphericalPoseReader(
    ActiveRegionScaleFactors=(PositiveWidth=2,NegativeWidth=0.5,PositiveHeight=2,NegativeHeight=2)
    ActiveRegionSize=0.15
    Debug=(bDrawDebug=false,bDraw2D=False,bDrawLocalAxes=False,DebugScale=25,DebugSegments=20,DebugThickness=0.25)
    DriverAxis=(X=1,Y=0,Z=0)
    DriverItem=(Type=Bone,Name="thigh_r")
    FalloffRegionScaleFactors=(PositiveWidth=2,NegativeWidth=1.8,PositiveHeight=2,NegativeHeight=2)
    FalloffSize=0.35
    FlipHeightScaling=False
    FlipWidthScaling=False
    OptionalParentItem=(Type=Bone,Name="None")
    RotationOffset=(X=-90,Y=180,Z=-10)
   )
   > &RigUnit_ModifyTransforms(
    ItemToModify="((Item=(Type=Bone,Name=\"base_skirt_front_c\"),Transform=(Rotation=(X=0.1106,Y=-0.1106,Z=-0.6984,W=0.6984),Translation=(X=-22.0017,Y=0.0001,Z=-0.0018),Scale3D=(X=1,Y=1,Z=1))),(Item=(Type=Bone,Name=\"base_skirt_side_r\"),Transform=(Rotation=(X=0,Y=0.5797,Z=0,W= ...(truncated)"
    WeightMinimum=0
    Weight=@n10.OutputParam
   )
   Multiply = &RigVMFunction_MathDoubleMul(A=@n9.OutputParam, B=@n10.OutputParam)
   > &RigUnit_ModifyTransforms(
    ItemToModify=((Item=(Type=Bone,Name="base_skirt_front_c"),Transform=(Rotation=(X=-0.0255,Y=0.2499,Z=0.7201,W=0.6468),Translation=(X=2.1309,Y=9.7946,Z=-3.0479),Scale3D=(X=1,Y=1,Z=1))))
    WeightMinimum=0
    Weight=@Multiply.Result
   )
   RigVMFunctionReturnNode(ExecuteContext=@RigVMFunction_Sequence.E)
   &RigUnit_ModifyTransforms(
    ExecutePin=()
    ItemToModify="((Item=(Type=Bone,Name=\"base_strings_01_r\"),Transform=(Rotation=(X=-0.3211,Y=-0.4199,Z=0.1374,W=0.8377),Translation=(X=3.2404,Y=-0.5972,Z=1.5398),Scale3D=(X=1,Y=1,Z=1))),(Item=(Type=Bone,Name=\"base_strings_01_l\"),Transform=(Rotation=(X=0.2958,Y=-0.7097,Z=- ...(truncated)"
    Weight=1
    WeightMinimum=0
   )
   Remap = &RigVMFunction_MathDoubleRemap(SourceMaximum=0.2, SourceMinimum=0, TargetMaximum=1, TargetMinimum=0, Value=@Multiply.Result)
  ]}
 New Function_ContainedGraph {nodes [
   RigVMFunctionEntryNode()
   > &RigUnit_AimConstraintLocalSpaceOffset(
    Child=(Type=Bone,Name="dyn_leg_lights_front_01_l")
    Filter=(bX=True,bY=True,bZ=True)
    Parents=((Item=(Type=Bone,Name="dyn_Aim_leg_lights_front_01_l"),Weight=1))
    Weight=1
    WorldUp=(Target=(X=0,Y=0,Z=1),Kind=Direction,Space=(Type=Bone,Name="dyn_Aim_leg_lights_front_01_l"))
    bMaintainOffset=true
   )
   > &RigUnit_AimConstraintLocalSpaceOffset(
    Child=(Type=Bone,Name="dyn_arm_rights_03_r")
    Filter=(bX=True,bY=True,bZ=True)
    Parents=((Item=(Type=Bone,Name="dyn_Aim_arm_rights_03_r"),Weight=1))
    Weight=1
    WorldUp=(Target=(X=0,Y=0,Z=1),Kind=Direction,Space=(Type=Bone,Name="dyn_Aim_arm_rights_03_r"))
    bMaintainOffset=true
   )
   > &RigUnit_AimConstraintLocalSpaceOffset(
    Child=(Type=Bone,Name="dyn_spine_lights_front_01_04_c")
    Filter=(bX=True,bY=True,bZ=True)
    Parents=((Item=(Type=Bone,Name="dyn_Aim_spine_lights_front_01_04_c"),Weight=1))
    Weight=1
    WorldUp=(Target=(X=0,Y=0,Z=1),Kind=Direction,Space=(Type=Bone,Name="dyn_Aim_spine_lights_front_01_04_c"))
    bMaintainOffset=true
   )
   > &RigUnit_AimConstraintLocalSpaceOffset(
    Child=(Type=Bone,Name="dyn_spine_lights_front_01_03_c")
    Filter=(bX=True,bY=True,bZ=True)
    Parents=((Item=(Type=Bone,Name="dyn_Aim_spine_lights_front_01_03_c"),Weight=1))
    Weight=1
    WorldUp=(Target=(X=0,Y=0,Z=1),Kind=Direction,Space=(Type=Bone,Name="dyn_Aim_spine_lights_front_01_03_c"))
    bMaintainOffset=true
   )
   > RigVMFunctionReturnNode()
   &RigUnit_AimConstraintLocalSpaceOffset(
    Child=(Type=Bone,Name="dyn_leg_lights_front_02_l")
    ExecutePin=()
    Filter=(bX=True,bY=True,bZ=True)
    Parents=((Item=(Type=Bone,Name="dyn_Aim_leg_lights_front_02_l"),Weight=1))
    Weight=1
    WorldUp=(Target=(X=0,Y=0,Z=1),Kind=Direction,Space=(Type=Bone,Name="dyn_Aim_leg_lights_front_02_l"))
    bMaintainOffset=true
   )
   &RigUnit_AimConstraintLocalSpaceOffset(
    Child=(Type=Bone,Name="dyn_sphere_02_r")
    ExecutePin=()
    Filter=(bX=True,bY=True,bZ=True)
    Parents=((Item=(Type=Bone,Name="dyn_Aim_sphere_02_r"),Weight=1))
    Weight=1
    WorldUp=(Target=(X=0,Y=0,Z=1),Kind=Direction,Space=(Type=Bone,Name="dyn_Aim_sphere_02_r"))
    bMaintainOffset=true
   )
   &RigUnit_AimConstraintLocalSpaceOffset(
    Child=(Type=Bone,Name="dyn_arm_lights_03_l")
    ExecutePin=()
    Filter=(bX=True,bY=True,bZ=True)
    Parents=((Item=(Type=Bone,Name="dyn_Aim_arm_lights_03_l"),Weight=1))
    Weight=1
    WorldUp=(Target=(X=0,Y=0,Z=1),Kind=Direction,Space=(Type=Bone,Name="dyn_Aim_arm_lights_03_l"))
    bMaintainOffset=true
   )
   &RigUnit_AimConstraintLocalSpaceOffset(
    Child=(Type=Bone,Name="dyn_spine_lights_front_01_02_c")
    ExecutePin=()
    Filter=(bX=True,bY=True,bZ=True)
    Parents=((Item=(Type=Bone,Name="dyn_Aim_spine_lights_front_01_02_c"),Weight=1))
    Weight=1
    WorldUp=(Target=(X=0,Y=0,Z=1),Kind=Direction,Space=(Type=Bone,Name="dyn_Aim_spine_lights_front_01_02_c"))
    bMaintainOffset=true
   )
  ]}
 RigVMFunctionLibrary {nodes [
   Aims to spheres and lights = RigVMCollapseNode(ExecuteContext=())
   Controlers = RigVMCollapseNode(Debug=False, ExecuteContext=())
   Legacy Control Rig = RigVMCollapseNode(ExecuteContext=())
   SPR Legs and Knees = RigVMCollapseNode(ExecuteContext=())
   Surface Constraints = RigVMCollapseNode(ExecuteContext=())
  ]}
 RigVMFunction_Sequence_1_1_ContainedGraph {nodes [
   n1 = RigVMFunctionEntryNode()
   n2 = &RigVMFunction_Sequence(ExecuteContext=@n1.ExecuteContext)
   n3 = &RigVMFunction_Sequence(ExecuteContext=@n2.B)
   n4 = &RigVMFunction_Sequence(ExecuteContext=@n3.B)
   n5 = &RigVMFunction_Sequence(ExecuteContext=@n4.B)
   RigVMFunctionReturnNode(A=@n2.A, B=@n3.A, E=@n4.A, F=@n5.A, G=@n5.B)
  ]}
 RigVMFunction_Sequence_1_ContainedGraph {nodes [
   n1 = RigVMFunctionEntryNode()
   RigVMFunction_Sequence = &RigVMFunction_Sequence(ExecuteContext=@n1.ExecuteContext)
   n2 = &RigVMFunction_Sequence(ExecuteContext=@RigVMFunction_Sequence.B)
   RigVMFunctionReturnNode(A=@RigVMFunction_Sequence.A, B=@n2.A, C=@n2.B)
  ]}
 RigVMFunction_Sequence_1_ContainedGraph {nodes [
   n1 = RigVMFunctionEntryNode()
   RigVMFunction_Sequence = &RigVMFunction_Sequence(ExecuteContext=@n1.ExecuteContext)
   n2 = &RigVMFunction_Sequence(ExecuteContext=@RigVMFunction_Sequence.B)
   n3 = &RigVMFunction_Sequence(ExecuteContext=@n2.B)
   n4 = &RigVMFunction_Sequence(ExecuteContext=@n3.B)
   RigVMFunctionReturnNode(A=@RigVMFunction_Sequence.A, B=@n2.A, C=@n3.A, D=@n4.A, E=@n4.B)
  ]}
 RigVMFunction_Sequence_2_1_ContainedGraph {nodes [
   n1 = RigVMFunctionEntryNode()
   n2 = &RigVMFunction_Sequence(ExecuteContext=@n1.ExecuteContext)
   n3 = &RigVMFunction_Sequence(ExecuteContext=@n2.B)
   n4 = &RigVMFunction_Sequence(ExecuteContext=@n3.B)
   n5 = &RigVMFunction_Sequence(ExecuteContext=@n4.B)
   n6 = &RigVMFunction_Sequence(ExecuteContext=@n5.B)
   n7 = &RigVMFunction_Sequence(ExecuteContext=@n6.B)
   n8 = &RigVMFunction_Sequence(ExecuteContext=@n7.B)
   RigVMFunctionReturnNode(A=@n2.A, B=@n3.A, C=@n4.A, D=@n5.A, E=@n6.A, F=@n7.A, G=@n8.A, H=@n8.B)
  ]}
 RigVMModel {nodes [
   RigUnit_BeginExecution = &RigUnit_BeginExecution()
   RigVMFunction_Sequence_2 = RigVMAggregateNode(ExecuteContext=@RigUnit_BeginExecution.ExecutePin)
   Controlers = RigVMFunctionReferenceNode(Debug=false, ExecuteContext=@RigVMFunction_Sequence_2.A)
   FN_DeformRig_FNC = RigVMFunctionReferenceNode(
    DeformRig_ Curve="(deformRig_l_upperarm_up_alpha_72_#1=False,deformRig_r_upperarm_up_alpha_73_#2=False,deformRig_l_upperarm_fr_alpha_74_#3=False,deformRig_r_upperarm_fr_alpha_75_CB51B8 ...(truncated)"
    DeformRig_Bool="(upperarm_1_#4=True,elbow_3_#5=True,wrist_7_#6=True,groin_8_#7=True,glute_17_#8=True,knee_9_40C408324CA261 ...(truncated)"
    DeformRig_Debug=(upperarm_debug_18_#4=False,wrist_debug_20_#6=False,lat_debug_26_#9=False,groinGlute_debug_27_#7=False)
    DeformRig_Value="(upperarm_up_18_#10=(Rotation=(X=0.0,Y=-0.0,Z=0.0,W=1.0),Translation=(X=0.0,Y=0.0,Z=8.0),Scale3D=(X=0.70,Y=1.0,Z=1.0)),upperarm_fwd_fr_19_#11=(Rotation=(X=0.0,Y=-0.0,Z=0.0,W=1.0),Translation=(X=0.0, ...(truncated)"
    ExecuteContext=@RigVMFunction_Sequence_2.B
   )
   Legacy Control Rig = RigVMFunctionReferenceNode(ExecuteContext=@RigVMFunction_Sequence_2.C)
   SPR Legs and Knees = RigVMFunctionReferenceNode(ExecuteContext=@RigVMFunction_Sequence_2.D)
   Surface Constraints = RigVMFunctionReferenceNode(ExecuteContext=@RigVMFunction_Sequence_2.E)
   Aims to spheres and lights = RigVMFunctionReferenceNode(ExecuteContext=@RigVMFunction_Sequence_2.F)
   &RigUnit_ParentConstraint(
    AdvancedSettings=(RotationOrderForFilter=XZY)
    Child=(Type=Bone,Name="base_Aim_strings_01_l")
    Parents="((Item=(Type=Bone,Name=\"skirt_side_02_l\"),Weight=0.0027),(Item=(Type=Bone,Name=\"spine_03\"),Weight=0.0034),(Item=(Type=Bone,Name=\"thigh_l\"),Weight=0.0038),(Item=(Type=Bone,Name=\"skirt_side_01_l\"),Weight=0.0122),(Item=(Type=Bone,Name=\"deform_groin_l\"),Weight ...(truncated)"
    Weight=1
    ExecutePin=@RigVMFunction_Sequence_2.G
   )
   > n1 = &RigUnit_ParentConstraint(
    AdvancedSettings=(RotationOrderForFilter=XZY)
    Child=(Type=Bone,Name="base_Aim_strings_01_r")
    Parents="((Item=(Type=Bone,Name=\"skirt_side_02_r\"),Weight=0.0013),(Item=(Type=Bone,Name=\"thigh_r\"),Weight=0.0013),(Item=(Type=Bone,Name=\"spine_02\"),Weight=0.016),(Item=(Type=Bone,Name=\"skirt_side_01_r\"),Weight=0.0224),(Item=(Type=Bone,Name=\"spine_01\"),Weight=0.026) ...(truncated)"
    Weight=1
   )
   Not_Equals = RigVMDispatchNode(B=true, A=RigVMVariableNode() {variable=IsFrontEnd}.Value) {template="DISPATCH_RigVMDispatch_CoreNotEquals(in A,in B,out Result)"}
   RigVMFunction_ControlFlowBranch = &RigVMFunction_ControlFlowBranch(Condition=@Not_Equals.Result, ExecuteContext=@n1.ExecutePin)
   Get Normalized Distance Between = RigVMFunctionReferenceNode(
    Max Distance=50
    Min Distance=20
    Source=(Type=Bone,Name="helper_bell_c")
    Target=(Type=Bone,Name="hand_r")
    ExecuteContext=@RigVMFunction_ControlFlowBranch.True
   )
   > &RigUnit_ModifyTransforms(
    ItemToModify="((Item=(Type=Bone,Name=\"base_bell_c\"),Transform=(Rotation=(X=0.1239,Y=0.0314,Z=-0.0449,W=0.9908),Translation=(X=-1.1351,Y=-13.192,Z=1.8157),Scale3D=(X=1,Y=1,Z=1))),(Item=(Type=Bone,Name=\"base_bow_r\"),Transform=(Rotation=(X=0,Y=0,Z=-0.1736,W=0.9848),Transla ...(truncated)"
    WeightMinimum=0
    Weight=@"Get Normalized Distance Between".Result
   )
   > &RigUnit_ModifyTransforms(
    ItemToModify=((Item=(Type=Bone,Name="dyn_bell_c"),Transform=(Rotation=(X=0,Y=0,Z=0,W=1),Translation=(X=-5.3575,Y=0,Z=0),Scale3D=(X=1,Y=1,Z=1))))
    WeightMinimum=0
    Weight=@"Get Normalized Distance Between".Result
   )
   > &RigUnit_ModifyTransforms(
    ItemToModify=((Item=(Type=Bone,Name="dyn_sphere_02_r"),Transform=(Rotation=(X=0.0665,Y=-0.6653,Z=-0.1604,W=0.7261),Translation=(X=0,Y=0,Z=0),Scale3D=(X=1,Y=1,Z=1))))
    WeightMinimum=0
    Weight=RigVMDispatchNode(BoneName=root, Default=0, Name=deformRig_r_elbow_alpha) {template="DISPATCH_RigDispatch_GetAnimAttribute(in Name,in BoneName,in Default,out Value,out Found)"}.Value
   )
   > &RigUnit_ModifyTransforms(
    ItemToModify=((Item=(Type=Bone,Name="base_arm_lights_03_l"),Transform=(Rotation=(X=0.0665,Y=-0.6653,Z=-0.1604,W=0.7261),Translation=(X=0,Y=0,Z=0),Scale3D=(X=1,Y=1,Z=1))))
    WeightMinimum=0
    Weight=RigVMDispatchNode(BoneName=root, Default=0, Name=deformRig_l_elbow_alpha) {template="DISPATCH_RigDispatch_GetAnimAttribute(in Name,in BoneName,in Default,out Value,out Found)"}.Value
   )
   &RigUnit_ModifyTransforms(
    ItemToModify=((Item=(Type=Bone,Name="base_leg_lights_front_02_l"),Transform=(Rotation=(X=0,Y=-0.5736,Z=0,W=0.8192),Translation=(X=0,Y=0,Z=0),Scale3D=(X=1,Y=1,Z=1))))
    WeightMinimum=0
    Weight=RigVMDispatchNode(BoneName=root, Default=0, Name=deformRig_l_thigh_up_alpha) {template="DISPATCH_RigDispatch_GetAnimAttribute(in Name,in BoneName,in Default,out Value,out Found)"}.Value
    ExecutePin=@RigVMFunction_Sequence_2.H
   )
   &RigUnit_ParentConstraint(
    AdvancedSettings=(RotationOrderForFilter=XZY)
    Child=(Type=Bone,Name="dyn_Aim_arm_lights_03_l")
    ExecutePin=()
    Parents="((Item=(Type=Bone,Name=\"upperarm_twist_02_l\"),Weight=0.0034),(Item=(Type=Bone,Name=\"upperarm_l\"),Weight=0.0236),(Item=(Type=Bone,Name=\"deform_elbow_in_l\"),Weight=0.025),(Item=(Type=Bone,Name=\"lowerarm_l\"),Weight=0.1558),(Item=(Type=Bone,Name=\"lowerarm_twis ...(truncated)"
    Weight=1
   )
   &RigUnit_ParentConstraint(
    AdvancedSettings=(RotationOrderForFilter=XZY)
    Child=(Type=Bone,Name="dyn_Aim_arm_lights_03_l")
    ExecutePin=()
    Parents="((Item=(Type=Bone,Name=\"upperarm_twist_02_l\"),Weight=0.0034),(Item=(Type=Bone,Name=\"upperarm_l\"),Weight=0.0236),(Item=(Type=Bone,Name=\"deform_elbow_in_l\"),Weight=0.025),(Item=(Type=Bone,Name=\"lowerarm_l\"),Weight=0.1558),(Item=(Type=Bone,Name=\"lowerarm_twis ...(truncated)"
    Weight=1
   )
   &RigUnit_ParentConstraint(
    AdvancedSettings=(RotationOrderForFilter=XZY)
    Child=(Type=Bone,Name="dyn_Aim_arm_lights_03_l")
    ExecutePin=()
    Parents="((Item=(Type=Bone,Name=\"upperarm_twist_02_l\"),Weight=0.0034),(Item=(Type=Bone,Name=\"upperarm_l\"),Weight=0.0236),(Item=(Type=Bone,Name=\"deform_elbow_in_l\"),Weight=0.025),(Item=(Type=Bone,Name=\"lowerarm_l\"),Weight=0.1558),(Item=(Type=Bone,Name=\"lowerarm_twis ...(truncated)"
    Weight=1
   )
   RigVMVariableNode() {variable=IsTargeting}
  ]}
]
hierarchy [
 root {loc=<0 0 0>, rot=<0 0 0 1>}
 pelvis {parent=root, loc=<0 -0.7321 92.4218>, rot=<0.0127 0.707 -0.0127 -0.707>}
 spine_01 {parent=pelvis, loc=<5.6002 -0.5128 0>, rot=<0 0 0.034 -0.9994>}
 spine_02 {parent=spine_01, loc=<8.0753 0.0139 0>, rot=<0 0 0.0238 -0.9997>}
 spine_03 {parent=spine_02, loc=<7.0393 0 0>, rot=<0 0 0.0149 -0.9999>}
 spine_04 {parent=spine_03, loc=<8.1024 0 0>, rot=<0 0 -0.0132 -0.9999>}
 spine_05 {parent=spine_04, loc=<10.0146 0 0>, rot=<0 0 -0.0015 -1>}
 clavicle_l {parent=spine_05, loc=<2.5856 -0.2866 -4.3777>, rot=<0.0465 -0.7289 -0.009 -0.6829>}
 upperarm_l {parent=clavicle_l, loc=<18.2264 0.1342 -0.001>, rot=<0.0253 -0.2693 0.0013 -0.9627>}
 lowerarm_l {parent=upperarm_l, loc=<29.0071 0.1686 -1.239>, rot=<0.0391 -0.0099 -0.2453 -0.9686>}
 hand_l {parent=lowerarm_l, loc=<25.8607 0.222 1.0274>, rot=<-0.7193 -0.0988 0.1193 -0.6773>}
 index_metacarpal_l {parent=hand_l, loc=<1.6535 0.0183 -2.96>, rot=<0.0158 -0.0124 -0.0173 -0.9996>}
 index_01_l {parent=index_metacarpal_l, loc=<8.3008 0.0357 -0.0524>, rot=<-0.0044 0.0485 0.0924 -0.9945>}
 index_02_l {parent=index_01_l, loc=<3.3587 -0.3004 0.0452>, rot=<-0.002 0.0027 0.3225 -0.9466>}
 index_03_l {parent=index_02_l, loc=<2.5578 -0.123 -0.0252>, rot=<-0.002 0.0067 0.2558 -0.9667>}
 middle_metacarpal_l {parent=hand_l, loc=<2.1856 -0.1725 -0.5984>, rot=<0.006 0.0168 -0.0329 -0.9993>}
 middle_01_l {parent=middle_metacarpal_l, loc=<7.3275 -0.2447 0.0129>, rot=<-0.0208 0.0042 0.0858 -0.9961>}
 middle_02_l {parent=middle_01_l, loc=<4.0123 0.1401 0.1175>, rot=<-0.0218 0.0258 0.3512 -0.9357>}
 middle_03_l {parent=middle_02_l, loc=<2.6077 0.0534 0.1373>, rot=<0.0082 0.0188 0.1708 -0.9851>}
 pinky_metacarpal_l {parent=hand_l, loc=<1.7616 -0.9753 3.3814>, rot=<-0.0397 0.0814 -0.069 -0.9935>}
 pinky_01_l {parent=pinky_metacarpal_l, loc=<7.1709 -0.4831 -0.1094>, rot=<-0.0654 -0.0035 0.2242 -0.9723>}
 pinky_02_l {parent=pinky_01_l, loc=<2.8274 0.0893 -0.1043>, rot=<0.0037 0.0011 0.2854 -0.9584>}
 pinky_03_l {parent=pinky_02_l, loc=<2.2046 0.0307 -0.0078>, rot=<-0.0002 0.0098 0.1445 -0.9895>}
 ring_metacarpal_l {parent=hand_l, loc=<2.2074 -0.3625 1.3638>, rot=<-0.0284 0.0722 -0.0458 -0.9959>}
 ring_01_l {parent=ring_metacarpal_l, loc=<7.0295 -0.6197 -0.1446>, rot=<-0.0489 -0.0252 0.1892 -0.9804>}
 thumb_01_l {parent=hand_l, loc=<1.3118 -0.3191 -2.459>, rot=<0.6546 -0.4042 0.0664 -0.6354>}
 thumb_02_l {parent=thumb_01_l, loc=<5.9336 -0.0448 0.0401>, rot=<-0.0037 0.0213 0.298 -0.9543>}
 thumb_03_l {parent=thumb_02_l, loc=<2.2223 0.3682 -0.0907>, rot=<0.0118 -0.013 0.0066 -0.9998>}
 deform_wrist_dn_base_l {parent=hand_l, loc=<0 -3.4173 0>, rot=<0 0 0 -1>}
 deform_wrist_dn_l {parent=deform_wrist_dn_base_l, loc=<0 0 0>, rot=<0 0 0 1>}
 deform_wrist_up_base_l {parent=hand_l, loc=<0 2.7805 0>, rot=<0 0 0 -1>}
 deform_wrist_up_l {parent=deform_wrist_up_base_l, loc=<0 0 0>, rot=<0 0 0 1>}
 lowerarm_twist_01_l {parent=lowerarm_l, loc=<17.8549 0 0>, rot=<0 0 0 1>}
 lowerarm_twist_02_l {parent=lowerarm_l, loc=<9.6199 0 0>, rot=<0 0 0 1>}
 deform_elbow_in_base_l {parent=lowerarm_l, loc=<0.2112 5.645 0.8274>, rot=<0 0 0 1>}
 deform_elbow_in_l {parent=deform_elbow_in_base_l, loc=<0 0 0>, rot=<0 0 0 1>}
 deform_elbow_out_base_l {parent=lowerarm_l, loc=<0 -7.5129 0>, rot=<0 0 0 1>}
 deform_elbow_out_l {parent=deform_elbow_out_base_l, loc=<0 0 0>, rot=<0 0 0 1>}
 upperarm_twist_01_l {parent=upperarm_l, loc=<7.9796 0.0843 -0.6195>, rot=<0 0 0 1>}
 deform_upperarm_bk_base_l {parent=upperarm_twist_01_l, loc=<-6.8983 -13.3826 -1.3258>, rot=<0 0 0 1>}
 deform_upperarm_bk_l {parent=deform_upperarm_bk_base_l, loc=<0 0 0>, rot=<0 0 0 1>}
 deform_upperarm_fr_base_l {parent=upperarm_twist_01_l, loc=<-8.3487 10.7117 -2.1455>, rot=<0 0 0 1>}
 deform_upperarm_fr_l {parent=deform_upperarm_fr_base_l, loc=<0 0 0>, rot=<0 0 0 1>}
 upperarm_twist_02_l {parent=upperarm_l, loc=<19.5059 0.0843 -0.6195>, rot=<0 0 0 1>}
 base_arm_lights_04_l {parent=upperarm_l, loc=<31.9091 -10.6941 -2.2193>, rot=<-0.6582 -0.0255 0.3044 -0.6881>}
 dyn_arm_lights_04_l {parent=base_arm_lights_04_l, loc=<0 0 0>, rot=<0 0 0 1>}
 base_arm_lights_06_l {parent=upperarm_l, loc=<28.5853 -9.1115 -8.8813>, rot=<-0.8051 -0.0362 0.2713 -0.5262>}
 dyn_arm_lights_06_l {parent=base_arm_lights_06_l, loc=<0 0 0>, rot=<0 0 0 1>}
 base_arm_lights_01_l {parent=upperarm_l, loc=<29.3598 12.9024 9.2485>, rot=<0.0098 0.3646 -0.1418 -0.9202>}
 dyn_arm_lights_01_l {parent=base_arm_lights_01_l, loc=<0 0 0>, rot=<0 0 0 1>}
 base_arm_lights_02_l {parent=upperarm_l, loc=<34.7675 -7.4949 7.1438>, rot=<-0.4054 -0.0589 0.269 -0.8717>}
 dyn_arm_lights_02_l {parent=base_arm_lights_02_l, loc=<0 0 0>, rot=<0 0 0 1>}
 base_arm_lights_03_l {parent=upperarm_l, loc=<24.1789 15.995 2.2854>, rot=<0.463 -0.1172 -0.4814 -0.735>}
 dyn_arm_lights_03_l {parent=base_arm_lights_03_l, loc=<0 0 0>, rot=<0 0 0 1>}
 base_arm_lights_07_l {parent=upperarm_l, loc=<19.5161 13.4751 -6.76>, rot=<0.4977 -0.1598 -0.4575 -0.7194>}
 dyn_arm_lights_07_l {parent=base_arm_lights_07_l, loc=<0 0 0>, rot=<0 0 0 1>}
 dyn_clavPush_l {parent=clavicle_l, loc=<0 0 0>, rot=<0 0 0 1>}
 clavicle_r {parent=spine_05, loc=<2.5856 -0.2865 4.3777>, rot=<0.7289 0.0465 -0.6829 0.009>}
 upperarm_r {parent=clavicle_r, loc=<-18.2264 -0.1342 0.001>, rot=<0.0253 -0.2693 0.0013 -0.9627>}
 lowerarm_r {parent=upperarm_r, loc=<-29.0072 -0.1686 1.239>, rot=<0.0391 -0.0099 -0.2453 -0.9686>}
 hand_r {parent=lowerarm_r, loc=<-25.8608 -0.222 -1.0274>, rot=<-0.7193 -0.0988 0.1193 -0.6773>}
 index_metacarpal_r {parent=hand_r, loc=<-1.6534 -0.0181 2.9601>, rot=<0.0158 -0.0124 -0.0173 -0.9996>}
 index_01_r {parent=index_metacarpal_r, loc=<-8.3008 -0.0357 0.0524>, rot=<-0.0044 0.0485 0.0924 -0.9945>}
 index_02_r {parent=index_01_r, loc=<-3.3589 0.3003 -0.0453>, rot=<-0.002 0.0027 0.3225 -0.9466>}
 index_03_r {parent=index_02_r, loc=<-2.5578 0.123 0.0252>, rot=<-0.002 0.0067 0.2558 -0.9667>}
 middle_metacarpal_r {parent=hand_r, loc=<-2.1856 0.1726 0.5984>, rot=<0.006 0.0168 -0.0329 -0.9993>}
 middle_01_r {parent=middle_metacarpal_r, loc=<-7.3273 0.2447 -0.0129>, rot=<-0.0208 0.0042 0.0858 -0.9961>}
 middle_02_r {parent=middle_01_r, loc=<-4.0124 -0.14 -0.1175>, rot=<-0.0218 0.0258 0.3512 -0.9357>}
 middle_03_r {parent=middle_02_r, loc=<-2.6077 -0.0533 -0.1373>, rot=<0.0082 0.0188 0.1708 -0.9851>}
 pinky_metacarpal_r {parent=hand_r, loc=<-1.7617 0.9753 -3.3814>, rot=<-0.0397 0.0814 -0.069 -0.9935>}
 pinky_01_r {parent=pinky_metacarpal_r, loc=<-7.1707 0.4832 0.1094>, rot=<-0.0654 -0.0035 0.2242 -0.9723>}
 pinky_02_r {parent=pinky_01_r, loc=<-2.8274 -0.0892 0.1043>, rot=<0.0037 0.0011 0.2854 -0.9584>}
 pinky_03_r {parent=pinky_02_r, loc=<-2.2045 -0.031 0.0078>, rot=<-0.0002 0.0098 0.1445 -0.9895>}
 ring_metacarpal_r {parent=hand_r, loc=<-2.2073 0.3627 -1.3638>, rot=<-0.0284 0.0722 -0.0458 -0.9959>}
 ring_01_r {parent=ring_metacarpal_r, loc=<-7.0297 0.6197 0.1446>, rot=<-0.0489 -0.0252 0.1892 -0.9804>}
 thumb_01_r {parent=hand_r, loc=<-1.3119 0.3192 2.459>, rot=<0.6546 -0.4042 0.0664 -0.6354>}
 thumb_02_r {parent=thumb_01_r, loc=<-5.9335 0.0447 -0.0401>, rot=<-0.0037 0.0213 0.298 -0.9543>}
 thumb_03_r {parent=thumb_02_r, loc=<-2.2223 -0.3682 0.0907>, rot=<0.0118 -0.013 0.0066 -0.9998>}
 deform_wrist_dn_base_r {parent=hand_r, loc=<0.0001 3.4173 0>, rot=<0 0 0 -1>}
 deform_wrist_dn_r {parent=deform_wrist_dn_base_r, loc=<0 0 0>, rot=<0 0 0 -1>}
 deform_wrist_up_base_r {parent=hand_r, loc=<0 -2.7803 0>, rot=<0 0 0 -1>}
 deform_wrist_up_r {parent=deform_wrist_up_base_r, loc=<0 0 0>, rot=<0 0 0 -1>}
 lowerarm_twist_01_r {parent=lowerarm_r, loc=<-17.855 0 0>, rot=<0 0 0 1>}
 lowerarm_twist_02_r {parent=lowerarm_r, loc=<-9.6199 0 0>, rot=<0 0 0 1>}
 base_arm_rights_04_r {parent=lowerarm_twist_02_r, loc=<3.0932 -12.581 -8.4565>, rot=<0.1495 -0.8778 0.1048 -0.4429>}
 dyn_arm_rights_04_r {parent=base_arm_rights_04_r, loc=<0 0 0>, rot=<0 0 0 1>}
 base_arm_rights_05_r {parent=lowerarm_twist_02_r, loc=<-1.1831 -12.3276 7.4374>, rot=<0.272 -0.789 0.4658 0.2941>}
 dyn_arm_rights_05_r {parent=base_arm_rights_05_r, loc=<0 0 0>, rot=<0 0 0 1>}
 base_arm_rights_06_r {parent=lowerarm_twist_02_r, loc=<1.6657 12.0008 -3.6616>, rot=<-0.3671 -0.6716 -0.642 -0.0444>}
 dyn_arm_rights_06_r {parent=base_arm_rights_06_r, loc=<0 0 0>, rot=<0 0 0 1>}
 base_sphere_02_r {parent=lowerarm_twist_02_r, loc=<1.429 -14.7025 -2.8842>, rot=<0.2417 -0.6141 0.7459 -0.0894>}
 dyn_sphere_02_r {parent=base_sphere_02_r, loc=<0 0 0>, rot=<0 0 0 1>}
 deform_elbow_in_base_r {parent=lowerarm_r, loc=<-0.2114 -5.6449 -0.827>, rot=<0 0 0 1>}
 deform_elbow_in_r {parent=deform_elbow_in_base_r, loc=<0 0 0>, rot=<0 0 0 1>}
 deform_elbow_out_base_r {parent=lowerarm_r, loc=<-0.0001 7.5129 0.0002>, rot=<0 0 0 1>}
 deform_elbow_out_r {parent=deform_elbow_out_base_r, loc=<0 0 0>, rot=<0 0 0 1>}
 upperarm_twist_01_r {parent=upperarm_r, loc=<-7.9796 -0.0843 0.6195>, rot=<0 0 0 1>}
 dyn_upperArmVolume_r {parent=upperarm_twist_01_r, loc=<0 0 0>, rot=<0 0 0 1>}
 deform_upperarm_bk_base_r {parent=upperarm_twist_01_r, loc=<6.8984 13.3827 1.3257>, rot=<0 0 0 1>}
 deform_upperarm_bk_r {parent=deform_upperarm_bk_base_r, loc=<0 0 0>, rot=<0 0 0 1>}
 deform_upperarm_fr_base_r {parent=upperarm_twist_01_r, loc=<8.3485 -10.7117 2.1457>, rot=<0 0 0 1>}
 deform_upperarm_fr_r {parent=deform_upperarm_fr_base_r, loc=<0 0 0>, rot=<0 0 0 1>}
 upperarm_twist_02_r {parent=upperarm_r, loc=<-19.506 -0.0843 0.6195>, rot=<0 0 0 1>}
 base_sphere_01_r {parent=upperarm_twist_02_r, loc=<5.7813 14.4201 2.8516>, rot=<-0.4969 -0.7107 -0.4849 0.1135>}
 dyn_sphere_01_r {parent=base_sphere_01_r, loc=<0 0 0>, rot=<0 0 0 1>}
 base_arm_rights_01_r {parent=upperarm_twist_02_r, loc=<4.117 11.0603 -5.0653>, rot=<-0.1006 -0.8646 -0.429 -0.2415>}
 dyn_arm_rights_01_r {parent=base_arm_rights_01_r, loc=<0 0 0>, rot=<0 0 0 1>}
 base_arm_rights_02_r {parent=upperarm_twist_02_r, loc=<6.519 -0.2211 -10.1492>, rot=<0.0194 -0.9254 -0.0169 -0.3781>}
 dyn_arm_rights_02_r {parent=base_arm_rights_02_r, loc=<0 0 0>, rot=<-0.0004 0 -0.0004 -1>}
 base_arm_rights_03_r {parent=upperarm_twist_02_r, loc=<8.5694 -11.4989 -3.7479>, rot=<0.4365 -0.867 0.2004 -0.1325>}
 dyn_arm_rights_03_r {parent=base_arm_rights_03_r, loc=<0 0 0>, rot=<0 0 0 1>}
 dyn_clavPush_r {parent=clavicle_r, loc=<0 0 0>, rot=<0 0 0 1>}
 neck_01 {parent=spine_05, loc=<10.815 1.9196 0>, rot=<0 0 -0.1864 -0.9825>}
 neck_02 {parent=neck_01, loc=<4.7347 0.3045 0>, rot=<0 0 0.0363 -0.9993>}
 head {parent=neck_02, loc=<5.7648 0.5553 -0.0046>, rot=<0 0 0.115 -0.9934>}
 deform_pec_aim_base_l {parent=spine_04, loc=<5.0232 13.6719 -2.6508>, rot=<0.0672 -0.5201 0.3228 -0.7879>}
 deform_pec_aim_l {parent=deform_pec_aim_base_l, loc=<0 0 0>, rot=<0 0 0 1>}
 deform_pec_base_l {parent=deform_pec_aim_l, loc=<16.3279 0 0>, rot=<0 0 0 1>}
 deform_pec_l {parent=deform_pec_base_l, loc=<0 0 0>, rot=<0 0 0 1>}
 deform_pec_aim_base_r {parent=spine_04, loc=<5.023 13.6718 2.6508>, rot=<-0.5201 -0.0672 0.7879 0.3228>}
 deform_pec_aim_r {parent=deform_pec_aim_base_r, loc=<0 0 0>, rot=<0 0 0 1>}
 deform_pec_base_r {parent=deform_pec_aim_r, loc=<-16.3279 0 0>, rot=<0 0 0 -1>}
 deform_pec_r {parent=deform_pec_base_r, loc=<0 0 0.0002>, rot=<0 0 0 -1>}
 surface_base_bell_c {parent=spine_04, loc=<0.737 20.9237 -0.1398>, rot=<0.4788 0.5204 0.5204 0.4788>}
 base_bell_c {parent=surface_base_bell_c, loc=<0 0 0>, rot=<0 0 0 1>}
 base_bow_l {parent=base_bell_c, loc=<0.2055 -4.53 0>, rot=<0 0 0.7071 -0.7071>}
 dyn_bow_l {parent=base_bow_l, loc=<0 0 0>, rot=<0 0 0 1>}
 base_bow_r {parent=base_bell_c, loc=<-0.2538 3.4497 0>, rot=<0 0 -0.7071 -0.7071>}
 dyn_bow_r {parent=base_bow_r, loc=<0 0 0>, rot=<0 0 0 1>}
 dyn_bell_c {parent=base_bell_c, loc=<1 0 0>, rot=<0 0 0 1>}
 base_strings_01_r {parent=base_bell_c, loc=<2.5154 1.7156 -4.5838>, rot=<-0.0324 -0.5832 -0.1452 -0.7986>}
 dyn_strings_01_r {parent=base_strings_01_r, loc=<0 0 0>, rot=<0 0 0 1>}
 dyn_strings_02_r {parent=dyn_strings_01_r, loc=<12.6209 0 0>, rot=<0 0 0 1>}
 base_strings_01_l {parent=base_bell_c, loc=<2.3557 -1.2077 -4.3112>, rot=<0.051 -0.5881 0.1187 -0.7984>}
 dyn_strings_01_l {parent=base_strings_01_l, loc=<0 0 0>, rot=<0 0 0 1>}
 dyn_strings_02_l {parent=dyn_strings_01_l, loc=<10.4287 0 0>, rot=<0 0 0 1>}
 helper_bell_c {parent=spine_04, loc=<-0.518 18.283 -15.661>, rot=<0 0 0 1>}
 base_spine_lights_01_01_c {parent=spine_03, loc=<12.6359 -21.4148 -13.6425>, rot=<0.1381 -0.6935 0.0614 -0.7044>}
 dyn_spine_lights_01_01_c {parent=base_spine_lights_01_01_c, loc=<0 0 0>, rot=<0 0 0 1>}
 base_spine_lights_01_02_c {parent=spine_03, loc=<11.139 -23.5121 0.2235>, rot=<0.5266 -0.4719 0.4719 -0.5266>}
 dyn_spine_lights_01_02_c {parent=base_spine_lights_01_02_c, loc=<0 0 0>, rot=<0 0 0 1>}
 base_spine_lights_01_03_c {parent=spine_03, loc=<8.5232 -21.4409 15.5677>, rot=<0.5657 -0.4242 0.516 -0.4835>}
 dyn_spine_lights_01_03_c {parent=base_spine_lights_01_03_c, loc=<0 0 0>, rot=<0 0 0 1>}
 base_spine_lights_02_04_c {parent=spine_02, loc=<2.2076 -19.8273 16.5554>, rot=<0.4992 -0.075 0.8586 -0.0889>}
 dyn_spine_lights_02_04_c {parent=base_spine_lights_02_04_c, loc=<0 0 0>, rot=<0 0 0 1>}
 base_spine_lights_front_01_02_c {parent=spine_02, loc=<-1.0438 20.5849 12.022>, rot=<0.2173 0.5897 0.7623 0.1547>}
 dyn_spine_lights_front_01_02_c {parent=base_spine_lights_front_01_02_c, loc=<0 0 0>, rot=<0 0 0 1>}
 base_spine_lights_front_01_01_c {parent=spine_02, loc=<8.6859 16.0818 17.9235>, rot=<0.7762 0.1881 0.5686 0.1969>}
 dyn_spine_lights_front_01_01_c {parent=base_spine_lights_front_01_01_c, loc=<0 0 0>, rot=<0 0 0 1>}
 base_spine_lights_02_01_c {parent=spine_01, loc=<6.8459 -24.3077 8.8967>, rot=<0.508 -0.4919 0.4919 -0.508>}
 dyn_spine_lights_02_01_c {parent=base_spine_lights_02_01_c, loc=<0 0 0>, rot=<0 0 0 1>}
 base_spine_lights_02_02_c {parent=spine_01, loc=<3.9615 -25.7317 -5.3646>, rot=<0.1644 -0.8414 0.3027 -0.4163>}
 dyn_spine_lights_02_02_c {parent=base_spine_lights_02_02_c, loc=<0 0 0>, rot=<0 0 0 1>}
 base_spine_lights_02_03_c {parent=spine_01, loc=<-0.7042 -14.9385 -22.5946>, rot=<0.1293 -0.8779 0.2663 -0.3764>}
 dyn_spine_lights_02_03_c {parent=base_spine_lights_02_03_c, loc=<0 0 0>, rot=<0 0 0 1>}
 base_spine_lights_front_01_05_c {parent=spine_01, loc=<-4.3811 11.1079 -22.3867>, rot=<-0.3339 -0.6233 -0.3538 -0.6122>}
 dyn_spine_lights_front_01_05_c {parent=base_spine_lights_front_01_05_c, loc=<0 0 0>, rot=<0 0 0 1>}
 base_spine_lights_front_01_04_c {parent=spine_01, loc=<-2.2614 22.3115 -9.5139>, rot=<-0.1433 -0.8437 -0.3407 -0.3892>}
 dyn_spine_lights_front_01_04_c {parent=base_spine_lights_front_01_04_c, loc=<0 0 0>, rot=<0 0 0 1>}
 base_spine_lights_front_01_03_c {parent=spine_01, loc=<1.1101 24.1127 0.1833>, rot=<0.4184 0.6105 0.6724 -0.0113>}
 dyn_spine_lights_front_01_03_c {parent=base_spine_lights_front_01_03_c, loc=<0 0 0>, rot=<0 0 0 1>}
 thigh_l {parent=pelvis, loc=<-1.8716 -0.0109 -11.9936>, rot=<0.0012 0.0572 0.0155 -0.9982>}
 calf_l {parent=thigh_l, loc=<-41.9178 -0.2246 0.0926>, rot=<0.0006 -0.0161 -0.0391 -0.9991>}
 calf_twist_01_l {parent=calf_l, loc=<-32.3044 -0.236 0.1628>, rot=<-0.0712 0.0088 -0.0018 -0.9974>}
 calf_twist_02_l {parent=calf_l, loc=<-15.9268 -0.0967 0.0397>, rot=<0.0237 0.0095 -0.0035 -0.9997>}
 foot_l {parent=calf_l, loc=<-38.7057 1.2438 0.2452>, rot=<0.0975 -0.0474 0.0317 -0.9936>}
 ball_l {parent=foot_l, loc=<-6.1619 13.0403 0.6764>, rot=<-0.0235 0.0021 -0.6862 -0.727>}
 dyn_ball_l {parent=foot_l, loc=<-8.147 8.061 -0.375>, rot=<0.0001 -0.0001 -0.7063 -0.7079>}
 deform_knee_in_base_l {parent=calf_l, loc=<1.298 -8.0356 -0.0496>, rot=<0 0 0 1>}
 deform_knee_in_l {parent=deform_knee_in_base_l, loc=<0 0 0>, rot=<0 0 0 1>}
 deform_knee_out_base_l {parent=calf_l, loc=<0 6.8739 0>, rot=<0 0 0 1>}
 deform_knee_out_l {parent=deform_knee_out_base_l, loc=<0 0 0>, rot=<0 0 0 1>}
 thigh_twist_01_l {parent=thigh_l, loc=<-21.6303 0 0>, rot=<0.0013 0 0 -1>}
 deform_groin_base_l {parent=thigh_l, loc=<-2.1908 8.5661 -1.7442>, rot=<0 0 0 -1>}
 deform_groin_l {parent=deform_groin_base_l, loc=<0 0 0>, rot=<0 0 0 1>}
 base_leg_back_lights_01_l {parent=thigh_l, loc=<-7.3368 -20.4928 -7.518>, rot=<0.1262 -0.9019 0.3523 -0.2155>}
 dyn_leg_back_lights_01_l {parent=base_leg_back_lights_01_l, loc=<0 0 0>, rot=<0 0 0 1>}
 base_leg_back_lights_02_l {parent=thigh_l, loc=<-8.411 -25.7784 -1.1576>, rot=<0.211 -0.8308 0.3806 -0.3471>}
 dyn_leg_back_lights_02_l {parent=base_leg_back_lights_02_l, loc=<0 0 0>, rot=<0 0 0 1>}
 base_leg_lights_front_01_l {parent=thigh_l, loc=<3.7722 17.5141 -5.5179>, rot=<-0.1416 -0.8743 -0.4286 -0.1783>}
 dyn_leg_lights_front_01_l {parent=base_leg_lights_front_01_l, loc=<0 0 0>, rot=<0 0 0 1>}
 base_leg_lights_front_02_l {parent=thigh_l, loc=<-0.5645 12.5824 -10.0271>, rot=<-0.0475 -0.9384 -0.0923 -0.3296>}
 dyn_leg_lights_front_02_l {parent=base_leg_lights_front_02_l, loc=<0 0 0>, rot=<0 0 0 1>}
 thigh_r {parent=pelvis, loc=<-1.8714 -0.0109 11.9936>, rot=<0.0572 -0.0012 0.9982 0.0155>}
 calf_r {parent=thigh_r, loc=<41.918 0.2246 -0.0926>, rot=<0.0006 -0.0161 -0.0391 -0.9991>}
 calf_twist_01_r {parent=calf_r, loc=<32.5543 1.0004 -0.0817>, rot=<-0.0953 0.0001 -0.0075 -0.9954>}
 calf_twist_02_r {parent=calf_r, loc=<15.9268 0.0967 -0.0397>, rot=<0.0237 0.0095 -0.0035 -0.9997>}
 foot_r {parent=calf_r, loc=<38.7058 -1.2438 -0.2453>, rot=<0.0975 -0.0474 0.0317 -0.9936>}
 ball_r {parent=foot_r, loc=<6.2569 -13.2122 -0.7447>, rot=<0.0025 0.0101 -0.7146 -0.6995>}
 dyn_ball_r {parent=foot_r, loc=<8.147 -8.061 0.375>, rot=<0.0001 -0.0001 -0.7063 -0.7079>}
 deform_knee_in_base_r {parent=calf_r, loc=<-1.2981 8.0356 0.0496>, rot=<0 0 0 1>}
 deform_knee_in_r {parent=deform_knee_in_base_r, loc=<0 0 0>, rot=<0 0 0 1>}
 deform_knee_out_base_r {parent=calf_r, loc=<0 -6.8739 -0.0001>, rot=<0 0 0 1>}
 deform_knee_out_r {parent=deform_knee_out_base_r, loc=<0 0 0>, rot=<0 0 0 1>}
 thigh_twist_01_r {parent=thigh_r, loc=<21.6303 0 0>, rot=<0.0013 0 0 -1>}
 deform_groin_base_r {parent=thigh_r, loc=<2.191 -8.5661 1.7442>, rot=<0 0 0 -1>}
 deform_groin_r {parent=deform_groin_base_r, loc=<0 0 0>, rot=<0 0 0 -1>}
 base_leg_back_lights_02_r {parent=thigh_r, loc=<21.0612 26.4066 -2.1615>, rot=<0.7986 -0.0068 -0.0837 -0.596>}
 dyn_leg_back_lights_02_r {parent=base_leg_back_lights_02_r, loc=<0 0 0>, rot=<0 0 0 1>}
 base_leg_back_lights_01_r {parent=thigh_r, loc=<17.5766 27.8806 -9.4328>, rot=<0.744 0.1938 -0.3882 -0.5081>}
 dyn_leg_back_lights_01_r {parent=base_leg_back_lights_01_r, loc=<0 0 0>, rot=<0 0 0 1>}
 base_leg_back_lights_03_r {parent=thigh_r, loc=<29.5283 20.1946 8.2672>, rot=<0.422 0.1391 -0.1149 -0.8885>}
 dyn_leg_back_lights_03_r {parent=base_leg_back_lights_03_r, loc=<0 0 0>, rot=<0 0 0 1>}
 base_leg_lights_front_01_r {parent=thigh_r, loc=<33.1988 -16.7655 10.6197>, rot=<-0.6715 0.0506 -0.0338 -0.7385>}
 dyn_leg_lights_front_01_r {parent=base_leg_lights_front_01_r, loc=<0 0 0>, rot=<0 0 0 1>}
 base_leg_lights_front_02_r {parent=thigh_r, loc=<26.7337 -22.3148 -2.2705>, rot=<-0.7006 0.0742 0.1575 -0.692>}
 dyn_leg_lights_front_02_r {parent=base_leg_lights_front_02_r, loc=<0 0 0>, rot=<0 0 0 1>}
 base_leg_lights_front_03_r {parent=thigh_r, loc=<20.4022 -24.3543 -12.1032>, rot=<-0.2536 0.6804 0.3904 -0.566>}
 dyn_leg_lights_front_03_r {parent=base_leg_lights_front_03_r, loc=<0 0 0>, rot=<0 0 0 1>}
 deform_glute_base_l {parent=pelvis, loc=<-10.1877 -11.6217 -12.899>, rot=<0.0275 0.0501 -0.4472 -0.8926>}
 deform_glute_l {parent=deform_glute_base_l, loc=<0 0 0>, rot=<0 0 0 1>}
 deform_glute_base_r {parent=pelvis, loc=<-10.1877 -11.6218 12.899>, rot=<0.0501 -0.0275 0.8926 -0.4472>}
 deform_glute_r {parent=deform_glute_base_r, loc=<0 0 0>, rot=<0 0 0 1>}
 base_skirt_back_l {parent=pelvis, loc=<-5.6318 -17.427 -14.4754>, rot=<0.0367 -0.9993 0 0>}
 skirt_back_01_l {parent=base_skirt_back_l, loc=<0 0 0>, rot=<0 0 0 1>}
 skirt_back_02_l {parent=skirt_back_01_l, loc=<17.0004 0 0>, rot=<0.0006 -0.0161 -0.0391 -0.9991>}
 skirt_back_03_l {parent=skirt_back_02_l, loc=<17.0004 0 0>, rot=<0 0 0 1>}
 base_skirt_back_r {parent=pelvis, loc=<-5.6318 -17.427 14.475>, rot=<0.0367 -0.9993 0 0>}
 skirt_back_01_r {parent=base_skirt_back_r, loc=<0 0 0>, rot=<0 0 0 1>}
 skirt_back_02_r {parent=skirt_back_01_r, loc=<17.0004 0 0>, rot=<0.0006 -0.0161 -0.0391 -0.9991>}
 skirt_back_03_r {parent=skirt_back_02_r, loc=<17.0004 0 0>, rot=<0 0 0 1>}
 base_skirt_side_r {parent=pelvis, loc=<-6.9751 0.8508 29.3234>, rot=<0.0113 -0.707 0.707 0.0113>}
 skirt_side_01_r {parent=base_skirt_side_r, loc=<0 0 0>, rot=<0 0 0 1>}
 skirt_side_02_r {parent=skirt_side_01_r, loc=<17.0004 0 0>, rot=<0 0 0.0323 -0.9995>}
 skirt_side_03_r {parent=skirt_side_02_r, loc=<17.0004 0 0>, rot=<0 0 0.0323 -0.9995>}
 base_skirt_side_l {parent=pelvis, loc=<-6.9751 0.8508 -29.323>, rot=<0.0113 0.707 0.707 -0.0113>}
 skirt_side_01_l {parent=base_skirt_side_l, loc=<0 0 0>, rot=<0 0 0 1>}
 skirt_side_02_l {parent=skirt_side_01_l, loc=<17.0004 0 0>, rot=<0 0 0.0323 -0.9995>}
 skirt_side_03_l {parent=skirt_side_02_l, loc=<17.0004 0 0>, rot=<0 0 0.0323 -0.9995>}
 base_skirt_front_c {parent=pelvis, loc=<-6.9751 22.4766 0>, rot=<0 0.0003 0.9998 0.0179>}
 skirt_front_01_c {parent=base_skirt_front_c, loc=<0 0 0>, rot=<0 0 0 1>}
 skirt_front_02_c {parent=skirt_front_01_c, loc=<17.0004 0 0>, rot=<0 0 0 1>}
 skirt_front_03_c {parent=skirt_front_02_c, loc=<17.0004 0 0>, rot=<0 0 0.0323 -0.9995>}
 dyn_Aim_leg_lights_front_02_l {parent=root, loc=<28.5411 -55.9753 80.9436>, rot=<0.4974 -0.7174 0.1998 0.4449>}
 dyn_Aim_leg_lights_front_01_l {parent=root, loc=<11.6192 -44.03 56.821>, rot=<0.325 -0.9313 0.1383 -0.0891>}
 dyn_Aim_sphere_02_r {parent=root, loc=<-46.7057 25.5293 43.2484>, rot=<-0.4721 -0.3496 0.6621 -0.4653>}
 dyn_Aim_arm_lights_03_l {parent=root, loc=<49.152 -27.8896 217.7587>, rot=<-0.3832 0.1122 -0.6645 -0.6317>}
 dyn_Aim_arm_rights_03_r {parent=root, loc=<-28.5878 5.9299 106.6902>, rot=<0.7012 -0.2519 -0.5976 -0.2962>}
 dyn_Aim_spine_lights_front_01_04_c {parent=root, loc=<29.659 -20.6026 39.7997>, rot=<0.0412 0.0513 -0.4001 -0.9141>}
 dyn_Aim_spine_lights_front_01_03_c {parent=root, loc=<-1.3263 -11.7594 21.5323>, rot=<-0.3335 0.0346 0.8812 0.3333>}
 dyn_Aim_spine_lights_front_01_02_c {parent=root, loc=<-16.3258 -50.4317 2.8836>, rot=<-0.5875 0.2898 0.6854 0.318>}
 base_Aim_strings_01_l {parent=root, loc=<20.7736 -20.9428 86.9583>, rot=<0.3712 -0.9272 0.0459 0.0176>}
 base_Aim_strings_01_r {parent=root, loc=<-1.5838 12.5744 88.2621>, rot=<0.1761 -0.9801 -0.0041 -0.0919>}
 thigh_l_ctrl {controlType=EULER_TRANSFORM, type=Control, loc=<0 0 0>, rot=<0 0 0 1>}
 thigh_r_ctrl {controlType=EULER_TRANSFORM, type=Control, loc=<0 0 0>, rot=<0 0 0 1>}
]
variables [
 IsFrontEnd {type=bool, default=False}
 IsTargeting {type=bool, default=False}
 isCrouching {type=bool, default=False}
]
