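# Physics asset for the large male CometHoliday cosmetic body, covering its dangling strings, bell and front skirt.
# Left and right strings form two-link capsule chains from small base anchors, with narrow swing limits and zero twist limit.
# The bell is constrained to spine_04 with free swing and twist.
# Large capsules cover spine_04 and both thighs, while the two front skirt capsules and these thigh bodies have no constraints.
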PhysicsAsset /BRCosmetics/Characters/Player/Male/Large/Bodies/M_LRG_CometHoliday/Meshes/M_LRG_CometHoliday_Physics
bodies {
 base_bell_c [{shapeType=Capsule, length=1, radius=1, center=<0.1 0 0>}]
 base_strings_01_l [{shapeType=Capsule, length=0.505, radius=0.505, shapeName=base_strings_01_l_capsule, rotation=<rot 0 -90 0>}]
 base_strings_01_r [{shapeType=Capsule, length=0.505, radius=0.505, shapeName=base_strings_01_r_capsule, rotation=<rot 0 -90 0>}]
 dyn_strings_01_l [{
   shapeType=Capsule
   length=5.9618
   radius=3.0829
   shapeName=dyn_strings_01_l_capsule
   center=<8.3443 0.16 -0.1722>
   rotation=<rot 90 180 176.6775>
  }]
 dyn_strings_01_r [{
   shapeType=Capsule
   length=8.1964
   radius=3.0748
   shapeName=dyn_strings_01_r_capsule
   center=<9.0301 -0.1009 0>
   rotation=<rot 90 180 -171.8235>
  }]
 dyn_strings_02_l [{shapeType=Capsule, length=8.2258, radius=1.4715, shapeName=dyn_strings_02_l_capsule, center=<8.2339 -0.5428 0>, rotation=<rot 90 0 -57.5293>}]
 dyn_strings_02_r [{
   shapeType=Capsule
   length=7.0178
   radius=3.5076
   shapeName=dyn_strings_02_r_capsule
   center=<8.6042 0 2.0553>
   rotation=<rot 90 -172.875 -82.8749>
  }]
 skirt_front_01_c [{shapeType=Capsule, length=40.5804, radius=4.7654, shapeName=skirt_front_01_c_capsule, center=<13.3789 3.0899 0>, rotation=<rot 0 -90 0>}]
 skirt_front_02_c [{shapeType=Capsule, length=40.5804, radius=4.7654, shapeName=skirt_front_02_c_capsule, center=<9.4793 3.0899 0>, rotation=<rot 0 -90 0>}]
 spine_04 [{shapeType=Capsule, length=15.2376, radius=18.4964, center=<-17.6914 3.7888 0>}]
 thigh_l [{shapeType=Capsule, length=19.3996, radius=22.7091, shapeName=thigh_l_capsule, center=<-27.1968 0 0>, rotation=<rot 90 0 0>}]
 thigh_r [{shapeType=Capsule, length=19.3996, radius=22.7091, shapeName=thigh_l_capsule, center=<27.1968 -0.8109 0>, rotation=<rot 90 0 180>}]
}
constraints [
 {bone1Name=dyn_strings_02_r, bone2Name=dyn_strings_01_r, swing1LimitDegrees=15, swing2LimitDegrees=15, twistLimitDegrees=0}
 {bone1Name=dyn_strings_01_r, bone2Name=base_strings_01_r, swing1LimitDegrees=12, swing2LimitDegrees=15, twistLimitDegrees=0}
 {bone1Name=dyn_strings_02_l, bone2Name=dyn_strings_01_l, swing1LimitDegrees=15, swing2LimitDegrees=15, twistLimitDegrees=0}
 {bone1Name=dyn_strings_01_l, bone2Name=base_strings_01_l, swing1LimitDegrees=15, swing2LimitDegrees=20, twistLimitDegrees=0}
 {
  bone1Name=base_bell_c
  bone2Name=spine_04
  swing1LimitDegrees=45
  swing1Motion=Free
  swing2LimitDegrees=45
  swing2Motion=Free
  twistLimitDegrees=45
  twistMotion=Free
 }
]
