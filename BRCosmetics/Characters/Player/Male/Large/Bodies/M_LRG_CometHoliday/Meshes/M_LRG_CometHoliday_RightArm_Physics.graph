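# Physics asset for the right-arm mesh of the large male CometHoliday player body, with capsule bodies on the upper spine, clavicle, upper arm and two sphere bones.
# Constraints chain spine_04, clavicle_r, upperarm_r, base_sphere_01_r and dyn_sphere_01_r.
# The upper-arm and base-sphere joints rotate freely, while the dyn sphere has a tight swing limit and locked twist, which suggests a dangling accessory.
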PhysicsAsset /BRCosmetics/Characters/Player/Male/Large/Bodies/M_LRG_CometHoliday/Meshes/M_LRG_CometHoliday_RightArm_Physics
bodies {
 base_sphere_01_r [{shapeType=Capsule, length=0.505, radius=0.505, shapeName=base_sphere_01_r_capsule, rotation=<rot 0 -90 0>}]
 clavicle_r [{shapeType=Capsule, length=8.9259, radius=10.7874, center=<-13.9124 0 0>, rotation=<rot 90 0 0>}]
 dyn_sphere_01_r [{shapeType=Capsule, length=0.1087, radius=5.6459, shapeName=dyn_sphere_01_r_capsule, center=<8.8962 -3.6167 0.1992>}]
 spine_04 [{shapeType=Capsule, length=20.0242, radius=22.981, center=<-11.0377 -3.2322 1.7432>, rotation=<rot -90 179.9208 -179.9209>}]
 upperarm_r [{
   shapeType=Capsule
   length=17.7923
   radius=12.5149
   shapeName=upperarm_r_capsule
   center=<-10.9586 1.6234 3.992>
   rotation=<rot 85 -179.9995 -179.9996>
  }]
}
constraints [
 {bone1Name=clavicle_r, bone2Name=spine_04, swing1LimitDegrees=45, swing2LimitDegrees=45, twistLimitDegrees=45}
 {
  bone1Name=base_sphere_01_r
  bone2Name=upperarm_r
  swing1LimitDegrees=45
  swing1Motion=Free
  swing2LimitDegrees=45
  swing2Motion=Free
  twistLimitDegrees=45
  twistMotion=Free
 }
 {
  bone1Name=dyn_sphere_01_r
  bone2Name=base_sphere_01_r
  swing1LimitDegrees=20
  swing2LimitDegrees=20
  twistLimitDegrees=45
  twistMotion=Locked
 }
 {
  bone1Name=upperarm_r
  bone2Name=clavicle_r
  swing1LimitDegrees=45
  swing1Motion=Free
  swing2LimitDegrees=45
  swing2Motion=Free
  twistLimitDegrees=45
  twistMotion=Free
 }
]
