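# Post-process style Anim Blueprint for the M_LRG_CometHoliday body cosmetic, adding secondary motion to lights, bows and a bell.
# It copies the attached parent's pose through a Control Rig, then runs cached stages of AnimDynamics springs and additive bone offsets.
# Weighted multi-bone constraints make accessory base bones follow body and skirt bones.
# RigidBody simulation finishes the pose, choosing front-end or regular physics assets before a right-arm pass.
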AnimBlueprint /BRCosmetics/Characters/Player/Male/Large/Bodies/M_LRG_CometHoliday/Meshes/M_LRG_CometHoliday_AnimBP
target_skeleton=&/M_LRG_CometHoliday_Skeleton
graphs [
 AnimGraph {nodes [
   n1 = RigidBody!(
    ComponentPose=Local To Component(LocalPose=Use cached pose 'ConstraintsPose'())
    bAlphaBoolEnabled:bool
    OverrideWorldGravity:Vector="0,0,0"
    ExternalForce
   ) -> (Pose) {
    alphaInputType=Bool
    bClampLinearTranslationLimitToRefPose=true
    bOverrideWorldGravity=true
    overridePhysicsAsset=&/M_LRG_CometHoliday_Physics
    simulationSpace=BaseBoneSpace
    alphaBoolBlend {blendInTime=0.1, blendOption=Linear, blendOutTime=0.1}
    baseBoneRef {boneName=spine_05}
    componentLinearAccScale=<0.5 0.5 0.5>
    componentLinearVelScale=<0.5 0.5 0.5>
    simSpaceSettings {velocityScaleZ=0, worldAlpha=0.5}
   }
   n2 = Component To Local(ComponentPose=@n1)
   n3 = RigidBody(ComponentPose=Local To Component(LocalPose=Use cached pose 'ConstraintsPose'()), OverrideWorldGravity="0,0,0") {
    alphaInputType=Bool
    bClampLinearTranslationLimitToRefPose=true
    bOverrideWorldGravity=true
    overridePhysicsAsset=&/M_LRG_CometHoliday_FrontEnd_Physics
    simulationSpace=BaseBoneSpace
    alphaBoolBlend {blendInTime=0.1, blendOption=Linear, blendOutTime=0.1}
    baseBoneRef {boneName=spine_05}
    componentLinearAccScale=<0.5 0.5 0.5>
    componentLinearVelScale=<0.5 0.5 0.5>
    simSpaceSettings {velocityScaleZ=0, worldAlpha=0.5}
   }
   n4 = Component To Local(ComponentPose=@n3)
   n5 = Get bIsInFrontEnd()
   n6 = RigidBody(
    ComponentPose=Local To Component(LocalPose=Blend Poses by bool(BlendPose_0=@n4, BlendPose_1=@n2, bActiveValue=@n5) {blendType=Linear})
    OverrideWorldGravity="0,0,0"
   ) {
    alphaInputType=Bool
    bClampLinearTranslationLimitToRefPose=true
    bOverrideWorldGravity=true
    overridePhysicsAsset=&/M_LRG_CometHoliday_RightArm_Physics
    simulationSpace=BaseBoneSpace
    alphaBoolBlend {blendInTime=1, blendOption=Linear, blendOutTime=1}
    baseBoneRef {boneName=spine_05}
    componentLinearAccScale {x=0.2}
    componentLinearVelScale {x=0.2}
    simSpaceSettings {velocityScaleZ=0}
   }
   Output Pose(Result=Component To Local(ComponentPose=@n6)) {layerGroup=None}
   n7 = Control Rig!(
    Source=Copy Pose From Mesh() {bUseAttachedParent=true}
    Alpha
    isCrouching:bool=Get bIsCrouching()
    IsFrontEnd:bool=Get bIsInFrontEnd()
    IsTargeting:bool=Get bIsTargeting()
   ) -> (Pose) {
    alphaBoolBlend {blendOption=Linear}
    controlRigAssetReference {blueprintRigClass=&/M_LRG_CometHoliday_CtrlRig.M_LRG_CometHoliday_CtrlRig_C}
   }
   n8 = "Constraint - surface_base_bell_c (PPPPPPPP)"!(
    ComponentPose=Local To Component(LocalPose=@n7)
    Alpha=1
    ConstraintWeights_0=0.5196
    ConstraintWeights_1:float=0.1822
    ConstraintWeights_2:float=0.0955
    ConstraintWeights_3:float=0.0521
    ConstraintWeights_4:float=0.0484
    ConstraintWeights_5:float=0.0462
    ConstraintWeights_6:float=0.0298
    ConstraintWeights_7:float=0.0261
   ) -> (Pose) {
    alphaBoolBlend {blendOption=Linear}
    boneToModify {boneName=surface_base_bell_c}
    constraintSetup [
     {targetBone {boneName=spine_04}}
     {targetBone {boneName=spine_03}}
     {targetBone {boneName=spine_05}}
     {targetBone {boneName=deform_pec_r}}
     {targetBone {boneName=spine_02}}
     {targetBone {boneName=dyn_clavPush_r}}
     {targetBone {boneName=clavicle_l}}
     {targetBone {boneName=deform_pec_l}}
    ]
    constraintWeights [1,1,1,1,1,1,1,1]
   }
   Save cached pose 'ControlRigPose'(Pose=Component To Local(ComponentPose=@n8)) {cacheName=ControlRigPose}
   n9 = Anim Dynamics - Bone: dyn_arm_rights_02_r(ComponentPose=Local To Component(LocalPose=Use cached pose 'LegCache'()), Alpha=1, ExternalForce="0,0,0") {
    angularDampingOverride=1
    angularSpringConstant=20000
    bAngularSpring=true
    bOverrideAngularDamping=true
    bUseGravityOverride=true
    gravityScale=0.35
    simulationSpace=RootRelative
    alphaBoolBlend {blendOption=Linear}
    boundBone {boneName=dyn_arm_rights_02_r}
    componentAppliedLinearAccClamp=<3000 3000 3000>
    componentLinearAccScale=<0.8 0.8 0.8>
    componentLinearVelScale=<0.8 0.8 0.8>
    gravityOverride {z=-980}
    physicsBodyDefinitions [{
      boundBone {boneName=dyn_arm_rights_02_r}
      boxExtents=<20 30 30>
      constraintSetup {angularLimitsMax=<15 15 15>, angularLimitsMin=<-15 -30 -15>, angularTarget=<1 0 0>}
      localJointOffset=<2 0 0>
     }]
   }
   n10 = Anim Dynamics - Bone: dyn_arm_rights_03_r(ComponentPose=@n9, Alpha=1, ExternalForce="0,0,0") {
    angularDampingOverride=1
    angularSpringConstant=20000
    bAngularSpring=true
    bOverrideAngularDamping=true
    bUseGravityOverride=true
    gravityScale=0.35
    simulationSpace=RootRelative
    alphaBoolBlend {blendOption=Linear}
    boundBone {boneName=dyn_arm_rights_03_r}
    componentAppliedLinearAccClamp=<3000 3000 3000>
    componentLinearAccScale=<0.8 0.8 0.8>
    componentLinearVelScale=<0.8 0.8 0.8>
    gravityOverride {z=-980}
    physicsBodyDefinitions [{
      boundBone {boneName=dyn_arm_rights_03_r}
      boxExtents=<20 30 30>
      constraintSetup {angularLimitsMax=<15 15 15>, angularLimitsMin=<-15 -30 -15>, angularTarget=<1 0 0>}
      localJointOffset=<4 0 0>
     }]
   }
   n11 = Anim Dynamics - Bone: dyn_arm_rights_01_r(ComponentPose=@n10, Alpha=1, ExternalForce="0,0,0") {
    angularDampingOverride=1
    angularSpringConstant=20000
    bAngularSpring=true
    bOverrideAngularDamping=true
    bUseGravityOverride=true
    gravityScale=0.35
    simulationSpace=RootRelative
    alphaBoolBlend {blendOption=Linear}
    boundBone {boneName=dyn_arm_rights_01_r}
    componentAppliedLinearAccClamp=<3000 3000 3000>
    componentLinearAccScale=<0.8 0.8 0.8>
    componentLinearVelScale=<0.8 0.8 0.8>
    gravityOverride {z=-980}
    physicsBodyDefinitions [{
      boundBone {boneName=dyn_arm_rights_01_r}
      boxExtents=<20 30 30>
      constraintSetup {angularLimitsMax=<15 15 15>, angularLimitsMin=<-15 -15 -15>, angularTarget=<1 0 0>}
      localJointOffset=<4 0 0>
     }]
   }
   n12 = Anim Dynamics - Bone: dyn_arm_lights_07_l(ComponentPose=@n11, Alpha=1, ExternalForce="0,0,0") {
    angularDampingOverride=1
    angularSpringConstant=20000
    bAngularSpring=true
    bOverrideAngularDamping=true
    bUseGravityOverride=true
    gravityScale=0.35
    simulationSpace=RootRelative
    alphaBoolBlend {blendOption=Linear}
    boundBone {boneName=dyn_arm_lights_07_l}
    componentAppliedLinearAccClamp=<3000 3000 3000>
    componentLinearAccScale=<0.8 0.8 0.8>
    componentLinearVelScale=<0.8 0.8 0.8>
    gravityOverride {z=-980}
    physicsBodyDefinitions [{
      boundBone {boneName=dyn_arm_lights_07_l}
      boxExtents=<20 30 30>
      constraintSetup {angularLimitsMax=<0 5 15>, angularLimitsMin=<0 -5 0>, angularTarget=<1 0 0>}
      localJointOffset=<2 0 0>
     }]
   }
   n13 = Anim Dynamics - Bone: dyn_arm_lights_06_l(ComponentPose=@n12, Alpha=1, ExternalForce="0,0,0") {
    angularDampingOverride=1
    angularSpringConstant=20000
    bAngularSpring=true
    bOverrideAngularDamping=true
    bUseGravityOverride=true
    gravityScale=0.35
    simulationSpace=RootRelative
    alphaBoolBlend {blendOption=Linear}
    boundBone {boneName=dyn_arm_lights_06_l}
    componentAppliedLinearAccClamp=<3000 3000 3000>
    componentLinearAccScale=<0.8 0.8 0.8>
    componentLinearVelScale=<0.8 0.8 0.8>
    gravityOverride {z=-980}
    physicsBodyDefinitions [{
      boundBone {boneName=dyn_arm_lights_06_l}
      boxExtents=<20 30 30>
      constraintSetup {angularLimitsMax=<0 5 0>, angularLimitsMin=<0 -5 -15>, angularTarget=<1 0 0>}
      localJointOffset=<2 0 0>
     }]
   }
   n14 = Anim Dynamics - Bone: dyn_arm_lights_03_l(ComponentPose=@n13, Alpha=1, ExternalForce="0,0,0") {
    angularDampingOverride=1
    angularSpringConstant=20000
    bAngularSpring=true
    bOverrideAngularDamping=true
    bUseGravityOverride=true
    gravityScale=0.35
    simulationSpace=RootRelative
    alphaBoolBlend {blendOption=Linear}
    boundBone {boneName=dyn_arm_lights_03_l}
    componentAppliedLinearAccClamp=<3000 3000 3000>
    componentLinearAccScale=<0.8 0.8 0.8>
    componentLinearVelScale=<0.8 0.8 0.8>
    gravityOverride {z=-980}
    physicsBodyDefinitions [{
      boundBone {boneName=dyn_arm_lights_03_l}
      boxExtents=<20 30 30>
      constraintSetup {angularLimitsMax=<0 -5 10>, angularLimitsMin=<0 -15 0>, angularTarget=<1 0 0>}
      localJointOffset=<2 0 0>
     }]
   }
   n15 = Anim Dynamics - Bone: dyn_arm_lights_02_l(ComponentPose=@n14, Alpha=1, ExternalForce="0,0,0") {
    angularDampingOverride=1
    angularSpringConstant=20000
    bAngularSpring=true
    bOverrideAngularDamping=true
    bUseGravityOverride=true
    gravityScale=0.35
    simulationSpace=RootRelative
    alphaBoolBlend {blendOption=Linear}
    boundBone {boneName=dyn_arm_lights_02_l}
    componentAppliedLinearAccClamp=<3000 3000 3000>
    componentLinearAccScale=<0.8 0.8 0.8>
    componentLinearVelScale=<0.8 0.8 0.8>
    gravityOverride {z=-980}
    physicsBodyDefinitions [{
      boundBone {boneName=dyn_arm_lights_02_l}
      boxExtents=<20 30 30>
      constraintSetup {angularLimitsMax=<0 0 5>, angularLimitsMin=<0 -15 -5>, angularTarget=<1 0 0>}
      localJointOffset=<4 0 0>
     }]
   }
   n16 = Anim Dynamics - Bone: dyn_arm_lights_01_l(ComponentPose=@n15, Alpha=1, ExternalForce="0,0,0") {
    angularDampingOverride=1
    angularSpringConstant=20000
    bAngularSpring=true
    bOverrideAngularDamping=true
    bUseGravityOverride=true
    gravityScale=0.35
    simulationSpace=RootRelative
    alphaBoolBlend {blendOption=Linear}
    boundBone {boneName=dyn_arm_lights_01_l}
    componentAppliedLinearAccClamp=<3000 3000 3000>
    componentLinearAccScale=<0.8 0.8 0.8>
    componentLinearVelScale=<0.8 0.8 0.8>
    gravityOverride {z=-980}
    physicsBodyDefinitions [{
      boundBone {boneName=dyn_arm_lights_01_l}
      boxExtents=<20 30 30>
      constraintSetup {angularLimitsMax=<0 0 5>, angularLimitsMin=<0 -15 -5>, angularTarget=<1 0 0>}
      localJointOffset=<2 0 0>
     }]
   }
   n17 = Anim Dynamics - Bone: dyn_arm_lights_04_l(ComponentPose=@n16, Alpha=1, ExternalForce="0,0,0") {
    angularDampingOverride=1
    angularSpringConstant=20000
    bAngularSpring=true
    bOverrideAngularDamping=true
    bUseGravityOverride=true
    gravityScale=0.35
    simulationSpace=RootRelative
    alphaBoolBlend {blendOption=Linear}
    boundBone {boneName=dyn_arm_lights_04_l}
    componentAppliedLinearAccClamp=<3000 3000 3000>
    componentLinearAccScale=<0.8 0.8 0.8>
    componentLinearVelScale=<0.8 0.8 0.8>
    gravityOverride {z=-980}
    physicsBodyDefinitions [{
      boundBone {boneName=dyn_arm_lights_04_l}
      boxExtents=<30 30 30>
      constraintSetup {angularLimitsMax=<0 20 30>, angularLimitsMin=<0 -10 -30>, angularTarget=<1 0 0>}
      localJointOffset=<20 0 0>
     }]
   }
   Save cached pose 'Back Spine Cache'(Pose=Component To Local(ComponentPose=@n17)) {cacheName=Back Spine Cache}
   n18 = Anim Dynamics - Bone: dyn_spine_lights_front_01_01_c(ComponentPose=Local To Component(LocalPose=Use cached pose 'Back Spine Cache'()), Alpha=1, ExternalForce="0,0,0") {
    angularDampingOverride=1
    angularSpringConstant=20000
    bAngularSpring=true
    bOverrideAngularDamping=true
    bUseGravityOverride=true
    gravityScale=0.35
    simulationSpace=RootRelative
    alphaBoolBlend {blendOption=Linear}
    boundBone {boneName=dyn_spine_lights_front_01_01_c}
    componentAppliedLinearAccClamp=<3000 3000 3000>
    componentLinearAccScale=<0.8 0.8 0.8>
    componentLinearVelScale=<0.8 0.8 0.8>
    gravityOverride {z=-980}
    physicsBodyDefinitions [{
      boundBone {boneName=dyn_spine_lights_front_01_01_c}
      boxExtents=<20 30 30>
      constraintSetup {angularLimitsMax=<15 15 15>, angularLimitsMin=<-15 -15 -15>, angularTarget=<1 0 0>}
      localJointOffset=<4 0 0>
     }]
   }
   n19 = Anim Dynamics - Bone: dyn_spine_lights_front_01_02_c(ComponentPose=@n18, Alpha=1, ExternalForce="0,0,0") {
    angularDampingOverride=1
    angularSpringConstant=20000
    bAngularSpring=true
    bOverrideAngularDamping=true
    bUseGravityOverride=true
    gravityScale=0.35
    simulationSpace=RootRelative
    alphaBoolBlend {blendOption=Linear}
    boundBone {boneName=dyn_spine_lights_front_01_02_c}
    componentAppliedLinearAccClamp=<3000 3000 3000>
    componentLinearAccScale=<0.8 0.8 0.8>
    componentLinearVelScale=<0.8 0.8 0.8>
    gravityOverride {z=-980}
    physicsBodyDefinitions [{
      boundBone {boneName=dyn_spine_lights_front_01_02_c}
      boxExtents=<20 30 30>
      constraintSetup {angularLimitsMax=<15 15 15>, angularLimitsMin=<-15 -15 -15>, angularTarget=<1 0 0>}
      localJointOffset=<4 0 0>
     }]
   }
   n20 = Anim Dynamics - Bone: dyn_spine_lights_front_01_03_c(ComponentPose=@n19, Alpha=1, ExternalForce="0,0,0") {
    angularDampingOverride=1
    angularSpringConstant=20000
    bAngularSpring=true
    bOverrideAngularDamping=true
    bUseGravityOverride=true
    gravityScale=0.35
    simulationSpace=RootRelative
    alphaBoolBlend {blendOption=Linear}
    boundBone {boneName=dyn_spine_lights_front_01_03_c}
    componentAppliedLinearAccClamp=<3000 3000 3000>
    componentLinearAccScale=<0.8 0.8 0.8>
    componentLinearVelScale=<0.8 0.8 0.8>
    gravityOverride {z=-980}
    physicsBodyDefinitions [{
      boundBone {boneName=dyn_spine_lights_front_01_03_c}
      boxExtents=<20 30 30>
      constraintSetup {angularLimitsMax=<15 15 15>, angularLimitsMin=<-15 -15 -15>, angularTarget=<1 0 0>}
      localJointOffset=<4 0 0>
     }]
   }
   n21 = Anim Dynamics - Bone: dyn_spine_lights_front_01_04_c(ComponentPose=@n20, Alpha=1, ExternalForce="0,0,0") {
    angularDampingOverride=1
    angularSpringConstant=20000
    bAngularSpring=true
    bOverrideAngularDamping=true
    bUseGravityOverride=true
    gravityScale=0.35
    simulationSpace=RootRelative
    alphaBoolBlend {blendOption=Linear}
    boundBone {boneName=dyn_spine_lights_front_01_04_c}
    componentAppliedLinearAccClamp=<3000 3000 3000>
    componentLinearAccScale=<0.8 0.8 0.8>
    componentLinearVelScale=<0.8 0.8 0.8>
    gravityOverride {z=-980}
    physicsBodyDefinitions [{
      boundBone {boneName=dyn_spine_lights_front_01_04_c}
      boxExtents=<20 30 30>
      constraintSetup {angularLimitsMax=<15 15 15>, angularLimitsMin=<-15 -15 -15>, angularTarget=<1 0 0>}
      localJointOffset=<4 0 0>
     }]
   }
   n22 = Anim Dynamics - Bone: dyn_spine_lights_front_01_05_c(ComponentPose=@n21, Alpha=1, ExternalForce="0,0,0") {
    angularDampingOverride=1
    angularSpringConstant=20000
    bAngularSpring=true
    bOverrideAngularDamping=true
    bUseGravityOverride=true
    gravityScale=0.35
    simulationSpace=RootRelative
    alphaBoolBlend {blendOption=Linear}
    boundBone {boneName=dyn_spine_lights_front_01_05_c}
    componentAppliedLinearAccClamp=<3000 3000 3000>
    componentLinearAccScale=<0.8 0.8 0.8>
    componentLinearVelScale=<0.8 0.8 0.8>
    gravityOverride {z=-980}
    physicsBodyDefinitions [{
      boundBone {boneName=dyn_spine_lights_front_01_05_c}
      boxExtents=<20 30 30>
      constraintSetup {angularLimitsMax=<15 15 15>, angularLimitsMin=<-15 -15 -15>, angularTarget=<1 0 0>}
      localJointOffset=<4 0 0>
     }]
   }
   Save cached pose 'Front Spine Cache'(Pose=Component To Local(ComponentPose=@n22)) {cacheName=Front Spine Cache}
   n23 = Anim Dynamics - Bone: dyn_leg_lights_front_01_l(ComponentPose=Local To Component(LocalPose=Use cached pose 'Front Spine Cache'()), Alpha=1, ExternalForce="0,0,0") {
    angularDampingOverride=1
    angularSpringConstant=20000
    bAngularSpring=true
    bOverrideAngularDamping=true
    bUseGravityOverride=true
    gravityScale=0.35
    simulationSpace=RootRelative
    alphaBoolBlend {blendOption=Linear}
    boundBone {boneName=dyn_leg_lights_front_01_l}
    componentAppliedLinearAccClamp=<3000 3000 3000>
    componentLinearAccScale=<0.8 0.8 0.8>
    componentLinearVelScale=<0.8 0.8 0.8>
    gravityOverride {z=-980}
    physicsBodyDefinitions [{
      boundBone {boneName=dyn_leg_lights_front_01_l}
      boxExtents=<20 30 30>
      constraintSetup {angularLimitsMax=<15 15 15>, angularLimitsMin=<-15 -15 -15>, angularTarget=<1 0 0>}
      localJointOffset=<4 0 0>
     }]
   }
   n24 = Anim Dynamics - Bone: dyn_leg_lights_front_02_l(ComponentPose=@n23, Alpha=1, ExternalForce="0,0,0") {
    angularDampingOverride=1
    angularSpringConstant=20000
    bAngularSpring=true
    bOverrideAngularDamping=true
    bUseGravityOverride=true
    gravityScale=0.35
    simulationSpace=RootRelative
    alphaBoolBlend {blendOption=Linear}
    boundBone {boneName=dyn_leg_lights_front_02_l}
    componentAppliedLinearAccClamp=<3000 3000 3000>
    componentLinearAccScale=<0.8 0.8 0.8>
    componentLinearVelScale=<0.8 0.8 0.8>
    gravityOverride {z=-980}
    physicsBodyDefinitions [{
      boundBone {boneName=dyn_leg_lights_front_02_l}
      boxExtents=<20 30 30>
      constraintSetup {angularLimitsMax=<15 15 15>, angularLimitsMin=<-15 -15 -15>, angularTarget=<1 0 0>}
      localJointOffset=<4 0 0>
     }]
   }
   n25 = Anim Dynamics - Bone: dyn_leg_lights_front_01_r(ComponentPose=@n24, Alpha=1, ExternalForce="0,0,0") {
    angularDampingOverride=1
    angularSpringConstant=20000
    bAngularSpring=true
    bOverrideAngularDamping=true
    bUseGravityOverride=true
    gravityScale=0.35
    simulationSpace=RootRelative
    alphaBoolBlend {blendOption=Linear}
    boundBone {boneName=dyn_leg_lights_front_01_r}
    componentAppliedLinearAccClamp=<3000 3000 3000>
    componentLinearAccScale=<0.8 0.8 0.8>
    componentLinearVelScale=<0.8 0.8 0.8>
    gravityOverride {z=-980}
    physicsBodyDefinitions [{
      boundBone {boneName=dyn_leg_lights_front_01_r}
      boxExtents=<20 30 30>
      constraintSetup {angularLimitsMax=<15 15 15>, angularLimitsMin=<-15 -15 -15>, angularTarget=<1 0 0>}
      localJointOffset=<4 0 0>
     }]
   }
   n26 = Anim Dynamics - Bone: dyn_leg_lights_front_02_r(ComponentPose=@n25, Alpha=1, ExternalForce="0,0,0") {
    angularDampingOverride=1
    angularSpringConstant=20000
    bAngularSpring=true
    bOverrideAngularDamping=true
    bUseGravityOverride=true
    gravityScale=0.35
    simulationSpace=RootRelative
    alphaBoolBlend {blendOption=Linear}
    boundBone {boneName=dyn_leg_lights_front_02_r}
    componentAppliedLinearAccClamp=<3000 3000 3000>
    componentLinearAccScale=<0.8 0.8 0.8>
    componentLinearVelScale=<0.8 0.8 0.8>
    gravityOverride {z=-980}
    physicsBodyDefinitions [{
      boundBone {boneName=dyn_leg_lights_front_02_r}
      boxExtents=<20 30 30>
      constraintSetup {angularLimitsMax=<15 15 15>, angularLimitsMin=<-15 -15 -15>, angularTarget=<1 0 0>}
      localJointOffset=<4 0 0>
     }]
   }
   n27 = Anim Dynamics - Bone: dyn_leg_lights_front_03_r(ComponentPose=@n26, Alpha=1, ExternalForce="0,0,0") {
    angularDampingOverride=1
    angularSpringConstant=20000
    bAngularSpring=true
    bOverrideAngularDamping=true
    bUseGravityOverride=true
    gravityScale=0.35
    simulationSpace=RootRelative
    alphaBoolBlend {blendOption=Linear}
    boundBone {boneName=dyn_leg_lights_front_03_r}
    componentAppliedLinearAccClamp=<3000 3000 3000>
    componentLinearAccScale=<0.8 0.8 0.8>
    componentLinearVelScale=<0.8 0.8 0.8>
    gravityOverride {z=-980}
    physicsBodyDefinitions [{
      boundBone {boneName=dyn_leg_lights_front_03_r}
      boxExtents=<20 30 30>
      constraintSetup {angularLimitsMax=<15 15 15>, angularLimitsMin=<-15 -15 -15>, angularTarget=<1 0 0>}
      localJointOffset=<4 0 0>
     }]
   }
   Save cached pose 'Front Leg Cache'(Pose=Component To Local(ComponentPose=@n27)) {cacheName=Front Leg Cache}
   n28 = Anim Dynamics - Bone: dyn_spine_lights_01_01_c(ComponentPose=Local To Component(LocalPose=Use cached pose 'ControlRigPose'()), Alpha=1, ExternalForce="0,0,0") {
    angularDampingOverride=1
    angularSpringConstant=20000
    bAngularSpring=true
    bOverrideAngularDamping=true
    bUseGravityOverride=true
    gravityScale=0.35
    simulationSpace=RootRelative
    alphaBoolBlend {blendOption=Linear}
    boundBone {boneName=dyn_spine_lights_01_01_c}
    componentAppliedLinearAccClamp=<3000 3000 3000>
    componentLinearAccScale=<0.8 0.8 0.8>
    componentLinearVelScale=<0.8 0.8 0.8>
    gravityOverride {z=-980}
    physicsBodyDefinitions [{
      boundBone {boneName=dyn_spine_lights_01_01_c}
      boxExtents=<60 60 60>
      constraintSetup {angularLimitsMax=<0 40 5>, angularLimitsMin=<0 -30 -5>, angularTarget=<1 0 0>}
      localJointOffset=<3 0 0>
     }]
   }
   n29 = Anim Dynamics - Bone: dyn_spine_lights_01_02_c(ComponentPose=@n28, Alpha=1, ExternalForce="0,0,0") {
    angularDampingOverride=1
    angularSpringConstant=20000
    bAngularSpring=true
    bOverrideAngularDamping=true
    bUseGravityOverride=true
    gravityScale=0.35
    simulationSpace=RootRelative
    alphaBoolBlend {blendOption=Linear}
    boundBone {boneName=dyn_spine_lights_01_02_c}
    componentAppliedLinearAccClamp=<3000 3000 3000>
    componentLinearAccScale=<0.8 0.8 0.8>
    componentLinearVelScale=<0.8 0.8 0.8>
    gravityOverride {z=-980}
    physicsBodyDefinitions [{
      boundBone {boneName=dyn_spine_lights_01_02_c}
      boxExtents=<40 50 50>
      constraintSetup {angularLimitsMax=<0 20 5>, angularLimitsMin=<0 -20 -5>, angularTarget=<1 0 0>}
      localJointOffset=<5 0 0>
     }]
   }
   n30 = Anim Dynamics - Bone: dyn_spine_lights_01_03_c(ComponentPose=@n29, Alpha=1, ExternalForce="0,0,0") {
    angularDampingOverride=1
    angularSpringConstant=20000
    bAngularSpring=true
    bOverrideAngularDamping=true
    bUseGravityOverride=true
    gravityScale=0.35
    simulationSpace=RootRelative
    alphaBoolBlend {blendOption=Linear}
    boundBone {boneName=dyn_spine_lights_01_03_c}
    componentAppliedLinearAccClamp=<3000 3000 3000>
    componentLinearAccScale=<0.8 0.8 0.8>
    componentLinearVelScale=<0.8 0.8 0.8>
    gravityOverride {z=-980}
    physicsBodyDefinitions [{
      boundBone {boneName=dyn_spine_lights_01_03_c}
      boxExtents=<80 50 50>
      constraintSetup {angularLimitsMax=<0 20 5>, angularLimitsMin=<0 -20 -5>, angularTarget=<1 0 0>}
      localJointOffset=<4 0 0>
     }]
   }
   n31 = Anim Dynamics - Bone: dyn_spine_lights_02_01_c(ComponentPose=@n30, Alpha=1, ExternalForce="0,0,0") {
    angularDampingOverride=1
    angularSpringConstant=20000
    bAngularSpring=true
    bOverrideAngularDamping=true
    bUseGravityOverride=true
    gravityScale=0.35
    simulationSpace=RootRelative
    alphaBoolBlend {blendOption=Linear}
    boundBone {boneName=dyn_spine_lights_02_01_c}
    componentAppliedLinearAccClamp=<3000 3000 3000>
    componentLinearAccScale=<0.8 0.8 0.8>
    componentLinearVelScale=<0.8 0.8 0.8>
    gravityOverride {z=-980}
    physicsBodyDefinitions [{
      boundBone {boneName=dyn_spine_lights_02_01_c}
      boxExtents=<20 20 20>
      constraintSetup {angularLimitsMax=<0 20 5>, angularLimitsMin=<0 -20 -5>, angularTarget=<1 0 0>}
      localJointOffset=<7 0 0>
     }]
   }
   n32 = Anim Dynamics - Bone: dyn_spine_lights_02_02_c(ComponentPose=@n31, Alpha=1, ExternalForce="0,0,0") {
    angularDampingOverride=1
    angularSpringConstant=20000
    bAngularSpring=true
    bOverrideAngularDamping=true
    bUseGravityOverride=true
    gravityScale=0.35
    simulationSpace=RootRelative
    alphaBoolBlend {blendOption=Linear}
    boundBone {boneName=dyn_spine_lights_02_02_c}
    componentAppliedLinearAccClamp=<3000 3000 3000>
    componentLinearAccScale=<0.8 0.8 0.8>
    componentLinearVelScale=<0.8 0.8 0.8>
    gravityOverride {z=-980}
    physicsBodyDefinitions [{
      boundBone {boneName=dyn_spine_lights_02_02_c}
      boxExtents=<40 20 20>
      constraintSetup {angularLimitsMax=<0 30 5>, angularLimitsMin=<0 -20 -5>, angularTarget=<1 0 0>}
      localJointOffset=<2 0 0>
     }]
   }
   n33 = Anim Dynamics - Bone: dyn_spine_lights_02_03_c(ComponentPose=@n32, Alpha=1, ExternalForce="0,0,0") {
    angularDampingOverride=1
    angularSpringConstant=20000
    bAngularSpring=true
    bOverrideAngularDamping=true
    bUseGravityOverride=true
    gravityScale=0.35
    simulationSpace=RootRelative
    alphaBoolBlend {blendOption=Linear}
    boundBone {boneName=dyn_spine_lights_02_03_c}
    componentAppliedLinearAccClamp=<3000 3000 3000>
    componentLinearAccScale=<0.8 0.8 0.8>
    componentLinearVelScale=<0.8 0.8 0.8>
    gravityOverride {z=-980}
    physicsBodyDefinitions [{
      boundBone {boneName=dyn_spine_lights_02_03_c}
      boxExtents=<40 30 50>
      constraintSetup {angularLimitsMax=<0 20 5>, angularLimitsMin=<0 -20 -5>, angularTarget=<1 0 0>}
      localJointOffset=<4 0 0>
     }]
   }
   n34 = Anim Dynamics - Bone: dyn_spine_lights_02_04_c(ComponentPose=@n33, Alpha=1, ExternalForce="0,0,0") {
    angularDampingOverride=1
    angularSpringConstant=20000
    bAngularSpring=true
    bOverrideAngularDamping=true
    bUseGravityOverride=true
    gravityScale=0.35
    simulationSpace=RootRelative
    alphaBoolBlend {blendOption=Linear}
    boundBone {boneName=dyn_spine_lights_02_04_c}
    componentAppliedLinearAccClamp=<3000 3000 3000>
    componentLinearAccScale=<0.8 0.8 0.8>
    componentLinearVelScale=<0.8 0.8 0.8>
    gravityOverride {z=-980}
    physicsBodyDefinitions [{
      boundBone {boneName=dyn_spine_lights_02_04_c}
      boxExtents=<40 30 50>
      constraintSetup {angularLimitsMax=<0 20 5>, angularLimitsMin=<0 -30 -5>, angularTarget=<1 0 0>}
      localJointOffset=<1 0 0>
     }]
   }
   Save cached pose 'SpineBackCache'(Pose=Component To Local(ComponentPose=@n34)) {cacheName=SpineBackCache}
   n35 = "Transform (Modify) Bone - Bone: base_bell_c"!(
    ComponentPose=Local To Component(LocalPose=Use cached pose 'SpineBackCache'())
    bAlphaBoolEnabled:bool=NOT Boolean(A=Get bIsInFrontEnd())
    Translation="0,0,5"
    Rotation="0,0,0"
    Scale="1,1,1"
   ) -> (Pose) {
    alphaInputType=Bool
    translationMode=BMM_Additive
    translationSpace=BCS_BoneSpace
    alphaBoolBlend {blendOption=Linear}
    boneToModify {boneName=base_bell_c}
    translation {z=5}
   }
   n36 = Anim Dynamics - Bone: dyn_leg_back_lights_02_l(ComponentPose=@n35, Alpha=1, ExternalForce="0,0,0") {
    angularDampingOverride=1
    angularSpringConstant=20000
    bAngularSpring=true
    bOverrideAngularDamping=true
    bUseGravityOverride=true
    gravityScale=0.35
    simulationSpace=RootRelative
    alphaBoolBlend {blendOption=Linear}
    boundBone {boneName=dyn_leg_back_lights_02_l}
    componentAppliedLinearAccClamp=<3000 3000 3000>
    componentLinearAccScale=<0.8 0.8 0.8>
    componentLinearVelScale=<0.8 0.8 0.8>
    gravityOverride {z=-980}
    physicsBodyDefinitions [{
      boundBone {boneName=dyn_leg_back_lights_02_l}
      boxExtents=<20 30 30>
      constraintSetup {angularLimitsMax=<0 0 10>, angularLimitsMin=<0 -15 -10>, angularTarget=<1 0 0>}
      localJointOffset=<4.02 0 0>
     }]
   }
   n37 = Anim Dynamics - Bone: dyn_leg_back_lights_01_l(ComponentPose=@n36, Alpha=1, ExternalForce="0,0,0") {
    angularDampingOverride=1
    angularSpringConstant=20000
    bAngularSpring=true
    bOverrideAngularDamping=true
    bUseGravityOverride=true
    gravityScale=0.35
    simulationSpace=RootRelative
    alphaBoolBlend {blendOption=Linear}
    boundBone {boneName=dyn_leg_back_lights_01_l}
    componentAppliedLinearAccClamp=<3000 3000 3000>
    componentLinearAccScale=<0.8 0.8 0.8>
    componentLinearVelScale=<0.8 0.8 0.8>
    gravityOverride {z=-980}
    physicsBodyDefinitions [{
      boundBone {boneName=dyn_leg_back_lights_01_l}
      boxExtents=<20 30 30>
      constraintSetup {angularLimitsMax=<0 0 30>, angularLimitsMin=<0 -15 -10>, angularTarget=<1 0 0>}
      localJointOffset=<10 0 0>
     }]
   }
   n38 = Anim Dynamics - Bone: dyn_leg_back_lights_01_r(ComponentPose=@n37, Alpha=1, ExternalForce="0,0,0") {
    angularDampingOverride=1
    angularSpringConstant=20000
    bAngularSpring=true
    bOverrideAngularDamping=true
    bUseGravityOverride=true
    gravityScale=0.35
    simulationSpace=RootRelative
    alphaBoolBlend {blendOption=Linear}
    boundBone {boneName=dyn_leg_back_lights_01_r}
    componentAppliedLinearAccClamp=<3000 3000 3000>
    componentLinearAccScale=<0.8 0.8 0.8>
    componentLinearVelScale=<0.8 0.8 0.8>
    gravityOverride {z=-980}
    physicsBodyDefinitions [{
      boundBone {boneName=dyn_leg_back_lights_01_r}
      boxExtents=<20 30 30>
      constraintSetup {angularLimitsMax=<0 0 10>, angularLimitsMin=<0 -15 -10>, angularTarget=<1 0 0>}
      localJointOffset=<4.02 0 0>
     }]
   }
   n39 = Anim Dynamics - Bone: dyn_leg_back_lights_02_r(ComponentPose=@n38, Alpha=1, ExternalForce="0,0,0") {
    angularDampingOverride=1
    angularSpringConstant=20000
    bAngularSpring=true
    bOverrideAngularDamping=true
    bUseGravityOverride=true
    gravityScale=0.35
    simulationSpace=RootRelative
    alphaBoolBlend {blendOption=Linear}
    boundBone {boneName=dyn_leg_back_lights_02_r}
    componentAppliedLinearAccClamp=<3000 3000 3000>
    componentLinearAccScale=<0.8 0.8 0.8>
    componentLinearVelScale=<0.8 0.8 0.8>
    gravityOverride {z=-980}
    physicsBodyDefinitions [{
      boundBone {boneName=dyn_leg_back_lights_02_r}
      boxExtents=<30 30 30>
      constraintSetup {angularLimitsMax=<0 0 10>, angularLimitsMin=<0 -15 -60>, angularTarget=<1 0 0>}
      localJointOffset=<10 0 0>
     }]
   }
   n40 = Anim Dynamics - Bone: dyn_leg_back_lights_03_r(ComponentPose=@n39, Alpha=1, ExternalForce="0,0,0") {
    angularDampingOverride=1
    angularSpringConstant=20000
    bAngularSpring=true
    bOverrideAngularDamping=true
    bUseGravityOverride=true
    gravityScale=0.35
    simulationSpace=RootRelative
    alphaBoolBlend {blendOption=Linear}
    boundBone {boneName=dyn_leg_back_lights_03_r}
    componentAppliedLinearAccClamp=<3000 3000 3000>
    componentLinearAccScale=<0.8 0.8 0.8>
    componentLinearVelScale=<0.8 0.8 0.8>
    gravityOverride {z=-980}
    physicsBodyDefinitions [{
      boundBone {boneName=dyn_leg_back_lights_03_r}
      boxExtents=<20 30 30>
      constraintSetup {angularLimitsMax=<0 0 10>, angularLimitsMin=<0 -20 -10>, angularTarget=<1 0 0>}
      localJointOffset=<4 0 0>
     }]
   }
   Save cached pose 'LegCache'(Pose=Component To Local(ComponentPose=@n40)) {cacheName=LegCache}
   n41 = Anim Dynamics - Bone: dyn_bow_l(ComponentPose=Local To Component(LocalPose=Use cached pose 'Front Leg Cache'()), Alpha=1, ExternalForce="0,0,0") {
    angularDampingOverride=1
    angularSpringConstant=20000
    bAngularSpring=true
    bOverrideAngularDamping=true
    bUseGravityOverride=true
    gravityScale=0.35
    simulationSpace=RootRelative
    alphaBoolBlend {blendOption=Linear}
    boundBone {boneName=dyn_bow_l}
    componentAppliedLinearAccClamp=<3000 3000 3000>
    componentLinearAccScale=<0.8 0.8 0.8>
    componentLinearVelScale=<0.8 0.8 0.8>
    gravityOverride {z=-980}
    physicsBodyDefinitions [{
      boundBone {boneName=dyn_bow_l}
      boxExtents=<20 30 30>
      constraintSetup {angularLimitsMax=<0 0 15>, angularLimitsMin=<0 -15 0>, angularTarget=<1 0 0>}
      localJointOffset=<4 0 0>
     }]
   }
   n42 = Anim Dynamics - Bone: dyn_bow_r(ComponentPose=@n41, Alpha=1, ExternalForce="0,0,0") {
    angularDampingOverride=1
    angularSpringConstant=20000
    bAngularSpring=true
    bOverrideAngularDamping=true
    bUseGravityOverride=true
    gravityScale=0.35
    simulationSpace=RootRelative
    alphaBoolBlend {blendOption=Linear}
    boundBone {boneName=dyn_bow_r}
    componentAppliedLinearAccClamp=<3000 3000 3000>
    componentLinearAccScale=<0.8 0.8 0.8>
    componentLinearVelScale=<0.8 0.8 0.8>
    gravityOverride {z=-980}
    physicsBodyDefinitions [{
      boundBone {boneName=dyn_bow_r}
      boxExtents=<20 30 30>
      constraintSetup {angularLimitsMax=<0 0 0>, angularLimitsMin=<0 -10 -15>, angularTarget=<1 0 0>}
      localJointOffset=<4 0 0>
     }]
   }
   n43 = Anim Dynamics - Bone: dyn_bell_c(ComponentPose=@n42, Alpha=1, ExternalForce="0,0,0") {
    angularDampingOverride=1
    angularSpringConstant=20000
    bAngularSpring=true
    bOverrideAngularDamping=true
    bUseGravityOverride=true
    gravityScale=0.35
    simulationSpace=RootRelative
    alphaBoolBlend {blendOption=Linear}
    boundBone {boneName=dyn_bell_c}
    componentAppliedLinearAccClamp=<3000 3000 3000>
    componentLinearAccScale=<0.8 0.8 0.8>
    componentLinearVelScale=<0.8 0.8 0.8>
    gravityOverride {z=-980}
    physicsBodyDefinitions [{
      boundBone {boneName=dyn_bell_c}
      boxExtents=<20 30 30>
      constraintSetup {angularLimitsMax=<0 15 15>, angularLimitsMin=<0 0 -15>, angularTarget=<1 0 0>}
      localJointOffset=<4 0 0>
     }]
   }
   Save cached pose 'BowAndBellCache'(Pose=Component To Local(ComponentPose=@n43)) {cacheName=BowAndBellCache}
   n44 = "Transform (Modify) Bone - Bone: attach_cape"!(
    ComponentPose=Local To Component(LocalPose=Use cached pose 'BowAndBellCache'())
    bAlphaBoolEnabled:bool=NOT Boolean(A=Get bRigidBodyEnabled())
    Translation="0,2,0"
    Rotation="-0.000,0.008,20.000"
   ) -> (Pose) {
    alphaInputType=Bool
    rotationMode=BMM_Additive
    rotationSpace=BCS_BoneSpace
    translationMode=BMM_Additive
    translationSpace=BCS_BoneSpace
    alphaBoolBlend {blendInTime=1, blendOption=Linear, blendOutTime=1}
    alphaScaleBiasClamp {bInterpResult=true}
    boneToModify {boneName=attach_cape}
    rotation=<rot 0 0.008 20>
    translation {y=2}
   }
   n45 = "Transform (Modify) Bone - Bone: base_skirt_front_c"!(ComponentPose=@n44, bAlphaBoolEnabled:bool=True, Rotation="0,-50,0") -> (Pose) {
    alphaInputType=Bool
    rotationMode=BMM_Additive
    rotationSpace=BCS_BoneSpace
    alphaBoolBlend {blendInTime=1, blendOption=Linear, blendOutTime=1}
    alphaScaleBiasClamp {bInterpResult=true}
    boneToModify {boneName=base_skirt_front_c}
    rotation {yaw=-50}
   }
   n46 = "Transform (Modify) Bone - Bone: base_skirt_back_l"!(ComponentPose=@n45, bAlphaBoolEnabled:bool=True, Rotation="35,0,0") -> (Pose) {
    alphaInputType=Bool
    rotationMode=BMM_Additive
    rotationSpace=BCS_BoneSpace
    alphaBoolBlend {blendInTime=1, blendOption=Linear, blendOutTime=1}
    boneToModify {boneName=base_skirt_back_l}
    rotation {pitch=35}
   }
   n47 = "Transform (Modify) Bone - Bone: base_skirt_back_r"!(ComponentPose=@n46, bAlphaBoolEnabled:bool=True, Rotation="-35,0,0") -> (Pose) {
    alphaInputType=Bool
    rotationMode=BMM_Additive
    rotationSpace=BCS_BoneSpace
    alphaBoolBlend {blendInTime=0.1, blendOption=Linear, blendOutTime=0.1}
    boneToModify {boneName=base_skirt_back_r}
    rotation {pitch=-35}
   }
   n48 = "Transform (Modify) Bone - Bone: base_bow_r"!(ComponentPose=@n47, bAlphaBoolEnabled:bool=True, Translation="0,0,0", Rotation="0,-20,0") -> (Pose) {
    alphaInputType=Bool
    rotationMode=BMM_Additive
    rotationSpace=BCS_BoneSpace
    translationMode=BMM_Additive
    translationSpace=BCS_BoneSpace
    alphaBoolBlend {blendInTime=0.1, blendOption=Linear, blendOutTime=0.1}
    boneToModify {boneName=base_bow_r}
    rotation {yaw=-20}
   }
   n49 = "Transform (Modify) Bone - Bone: base_bell_c"!(ComponentPose=@n48, bAlphaBoolEnabled=True, Translation="1,-3,0", Rotation="0,0,0") -> (Pose) {
    alphaInputType=Bool
    rotationMode=BMM_Additive
    rotationSpace=BCS_BoneSpace
    translationMode=BMM_Additive
    translationSpace=BCS_BoneSpace
    alphaBoolBlend {blendInTime=0.1, blendOption=Linear, blendOutTime=0.1}
    boneToModify {boneName=base_bell_c}
    translation=<1 -3>
   }
   n50 = "Transform (Modify) Bone - Bone: base_bell_c"(ComponentPose=@n49, bAlphaBoolEnabled=True, Translation="0,0,10", Rotation="0,0,0") {
    alphaInputType=Bool
    rotationMode=BMM_Additive
    rotationSpace=BCS_BoneSpace
    translationMode=BMM_Additive
    translationSpace=BCS_BoneSpace
    alphaBoolBlend {blendInTime=0.5, blendOption=Linear, blendOutTime=0.1}
    boneToModify {boneName=base_bell_c}
    translation {z=10}
   }
   Save cached pose 'ModifyTransformsCache'(Pose=Component To Local(ComponentPose=@n50)) {cacheName=ModifyTransformsCache}
   n51 = "Constraint - base_leg_back_lights_03_r (PPPPPPPP)"!(
    ComponentPose=Local To Component(LocalPose=Use cached pose 'ModifyTransformsCache'())
    Alpha=1
    ConstraintWeights_0=0.4354
    ConstraintWeights_1:float=0.4234
    ConstraintWeights_2:float=0.0789
    ConstraintWeights_3:float=0.0428
    ConstraintWeights_4:float=0.0073
    ConstraintWeights_5:float=0.0064
    ConstraintWeights_6:float=0.0054
    ConstraintWeights_7:float=0.0004
   ) -> (Pose) {
    alphaBoolBlend {blendOption=Linear}
    boneToModify {boneName=base_leg_back_lights_03_r}
    constraintSetup [
     {targetBone {boneName=skirt_side_02_r}}
     {targetBone {boneName=skirt_back_02_r}}
     {targetBone {boneName=skirt_back_01_r}}
     {targetBone {boneName=skirt_side_01_r}}
     {targetBone {boneName=thigh_twist_01_r}}
     {targetBone {boneName=pelvis}}
     {targetBone {boneName=skirt_back_03_r}}
     {targetBone {boneName=deform_glute_r}}
    ]
    constraintWeights [1,1,1,1,1,1,1,1]
   }
   n52 = "Constraint - base_spine_lights_front_01_02_c (PPPPPPPP)"!(
    ComponentPose=@n51
    Alpha=1
    ConstraintWeights_0=0.3901
    ConstraintWeights_1:float=0.3005
    ConstraintWeights_2:float=0.1447
    ConstraintWeights_3:float=0.0778
    ConstraintWeights_4:float=0.0562
    ConstraintWeights_5:float=0.0161
    ConstraintWeights_6:float=0.014
    ConstraintWeights_7:float=0.0005
   ) -> (Pose) {
    alphaBoolBlend {blendOption=Linear}
    boneToModify {boneName=base_spine_lights_front_01_02_c}
    constraintSetup [
     {targetBone {boneName=spine_03}}
     {targetBone {boneName=spine_02}}
     {targetBone {boneName=spine_04}}
     {targetBone {boneName=spine_01}}
     {targetBone {boneName=pelvis}}
     {targetBone {boneName=dyn_clavPush_r}}
     {targetBone {boneName=deform_pec_r}}
     {targetBone {boneName=deform_groin_r}}
    ]
    constraintWeights [1,1,1,1,1,1,1,1]
   }
   n53 = "Constraint - base_leg_back_lights_02_r (PPPPPPPP)"!(
    ComponentPose=@n52
    Alpha=1
    ConstraintWeights_0=0.6338
    ConstraintWeights_1:float=0.2415
    ConstraintWeights_2:float=0.049
    ConstraintWeights_3:float=0.0383
    ConstraintWeights_4:float=0.0217
    ConstraintWeights_5:float=0.0075
    ConstraintWeights_6:float=0.0053
    ConstraintWeights_7:float=0.003
   ) -> (Pose) {
    alphaBoolBlend {blendOption=Linear}
    boneToModify {boneName=base_leg_back_lights_02_r}
    constraintSetup [
     {targetBone {boneName=skirt_back_02_r}}
     {targetBone {boneName=skirt_back_01_r}}
     {targetBone {boneName=skirt_back_02_l}}
     {targetBone {boneName=skirt_side_02_r}}
     {targetBone {boneName=pelvis}}
     {targetBone {boneName=skirt_side_01_r}}
     {targetBone {boneName=skirt_back_01_l}}
     {targetBone {boneName=thigh_twist_01_r}}
    ]
    constraintWeights [1,1,1,1,1,1,1,1]
   }
   n54 = "Constraint - base_leg_back_lights_01_r (PPPPPPPP)"!(
    ComponentPose=@n53
    Alpha=1
    ConstraintWeights_0=0.4827
    ConstraintWeights_1:float=0.2871
    ConstraintWeights_2:float=0.0948
    ConstraintWeights_3:float=0.0634
    ConstraintWeights_4:float=0.0449
    ConstraintWeights_5:float=0.0177
    ConstraintWeights_6:float=0.006
    ConstraintWeights_7:float=0.0033
   ) -> (Pose) {
    alphaBoolBlend {blendOption=Linear}
    boneToModify {boneName=base_leg_back_lights_01_r}
    constraintSetup [
     {targetBone {boneName=skirt_back_01_r}}
     {targetBone {boneName=skirt_back_02_r}}
     {targetBone {boneName=skirt_back_02_l}}
     {targetBone {boneName=pelvis}}
     {targetBone {boneName=skirt_back_01_l}}
     {targetBone {boneName=spine_01}}
     {targetBone {boneName=thigh_twist_01_r}}
     {targetBone {boneName=skirt_side_02_r}}
    ]
    constraintWeights [1,1,1,1,1,1,1,1]
   }
   n55 = "Constraint - base_leg_back_lights_01_l (PPPPPPPP)"!(
    ComponentPose=@n54
    Alpha=1
    ConstraintWeights_0=0.4677
    ConstraintWeights_1:float=0.1841
    ConstraintWeights_2:float=0.1044
    ConstraintWeights_3:float=0.0912
    ConstraintWeights_4:float=0.0502
    ConstraintWeights_5:float=0.0482
    ConstraintWeights_6:float=0.0306
    ConstraintWeights_7:float=0.0236
   ) -> (Pose) {
    alphaBoolBlend {blendOption=Linear}
    boneToModify {boneName=base_leg_back_lights_01_l}
    constraintSetup [
     {targetBone {boneName=skirt_back_01_l}}
     {targetBone {boneName=pelvis}}
     {targetBone {boneName=skirt_side_01_l}}
     {targetBone {boneName=spine_01}}
     {targetBone {boneName=thigh_l}}
     {targetBone {boneName=thigh_twist_01_l}}
     {targetBone {boneName=spine_02}}
     {targetBone {boneName=spine_03}}
    ]
    constraintWeights [1,1,1,1,1,1,1,1]
   }
   n56 = "Transform (Modify) Bone - Bone: base_leg_back_lights_03_r"(ComponentPose=@n55, Alpha=1, Translation="0,0,0", Rotation="0,-20,0", Scale="1,1,1") {
    rotationMode=BMM_Additive
    rotationSpace=BCS_BoneSpace
    translationMode=BMM_Additive
    alphaBoolBlend {blendOption=Linear}
    boneToModify {boneName=base_leg_back_lights_03_r}
    rotation {yaw=-20}
   }
   n57 = "Transform (Modify) Bone - Bone: base_leg_back_lights_01_l"!(ComponentPose=@n56, bAlphaBoolEnabled:bool=True, Translation="0,0,0", Rotation="20,0,0", Scale="1,1,1") -> (Pose) {
    alphaInputType=Bool
    rotationMode=BMM_Additive
    rotationSpace=BCS_BoneSpace
    translationMode=BMM_Additive
    alphaBoolBlend {blendInTime=0.1, blendOption=Linear, blendOutTime=0.1}
    boneToModify {boneName=base_leg_back_lights_01_l}
    rotation {pitch=20}
   }
   n58 = "Transform (Modify) Bone - Bone: base_skirt_back_l"!(ComponentPose=@n57, bAlphaBoolEnabled=True, Translation:Vector="0,0,0", Rotation="0,10,0", Scale:Vector="1,1,1") -> (Pose) {
    alphaInputType=Bool
    rotationMode=BMM_Additive
    rotationSpace=BCS_BoneSpace
    translationMode=BMM_Additive
    alphaBoolBlend {blendInTime=0.5, blendOption=Linear, blendOutTime=0.5}
    boneToModify {boneName=base_skirt_back_l}
    rotation {yaw=10}
   }
   n59 = "Transform (Modify) Bone - Bone: base_skirt_back_r"!(ComponentPose=@n58, bAlphaBoolEnabled=True, Translation:Vector="0,0,0", Rotation="0,25,0", Scale:Vector="1,1,1") -> (Pose) {
    alphaInputType=Bool
    rotationMode=BMM_Additive
    rotationSpace=BCS_BoneSpace
    translationMode=BMM_Additive
    alphaBoolBlend {blendInTime=0.5, blendOption=Linear, blendOutTime=0.5}
    boneToModify {boneName=base_skirt_back_r}
    rotation {yaw=25}
   }
   n60 = "Transform (Modify) Bone - Bone: base_spine_lights_front_01_02_c"!(ComponentPose=@n59, bAlphaBoolEnabled:bool=True, Translation="0,0,0", Rotation="-0.003,80.038,25.121", Scale="1,1,1") -> (Pose) {
    alphaInputType=Bool
    rotationMode=BMM_Additive
    rotationSpace=BCS_BoneSpace
    alphaBoolBlend {blendInTime=0.2, blendOption=Linear, blendOutTime=0.2}
    boneToModify {boneName=base_spine_lights_front_01_02_c}
    rotation=<rot -0.003 80.038 25.121>
   }
   n61 = "Transform (Modify) Bone - Bone: base_arm_lights_03_l"!(ComponentPose=@n60, bAlphaBoolEnabled:bool=True, Translation="0,0,0", Rotation="28.032,-82.715,-11.027", Scale="1,1,1") -> (Pose) {
    alphaInputType=Bool
    rotationMode=BMM_Additive
    rotationSpace=BCS_BoneSpace
    translationMode=BMM_Additive
    translationSpace=BCS_BoneSpace
    alphaBoolBlend {blendOption=Linear}
    boneToModify {boneName=base_arm_lights_03_l}
    rotation=<rot 28.032 -82.715 -11.027>
   }
   n62 = "Transform (Modify) Bone - Bone: base_bell_c"!(ComponentPose=@n61, bAlphaBoolEnabled=True, Translation="1,0,8", Rotation="0,0,0", Scale:Vector="1,1,1") -> (Pose) {
    alphaInputType=Bool
    translationMode=BMM_Additive
    translationSpace=BCS_BoneSpace
    alphaBoolBlend {blendOption=Linear}
    boneToModify {boneName=base_bell_c}
    translation {x=1, z=8}
   }
   Save cached pose 'ConstraintsPose'(Pose=Component To Local(ComponentPose=@n62)) {cacheName=ConstraintsPose}
   Get bIsKneeSliding()
   Get bClothEnabled()
  ]}
 EventGraph {nodes [
   Event BlueprintUpdateAnimation()
   > n1 = Sequence!(execute) -> (then_0, then_1, then_2:Exec, then_3:Exec)
   Set isTacticalSprintingAndJumping(execute=@n1.then_2, isTacticalSprintingAndJumping:bool=AND Boolean(A=Get bIsTacticalSprinting(), B=Get bIsJumping())) -> (Output_Get:bool)
   n2 = NOT Boolean(A=Get bIsCrouching())
   n3 = Get bIsOnGrindRail()
   n4 = Get isCustomEmotePlaying() -> (isCustomEmotePlaying:bool)
   Set NotCrouching(
    execute=@n1.then_3
    NotCrouching:bool=XOR Boolean!(A:bool=OR Boolean!(A=@n2, B=@n3, C:bool) -> (ReturnValue), B:bool=@n4) -> (ReturnValue:bool)
   ) -> (Output_Get:bool)
   n5 = GetAnimInstance(self=GetOwnerSkeletalMeshComponent())
   Is Valid(exec=@n1.then_1, InputObject=@n5)
   > Set isCustomEmotePlaying(isCustomEmotePlaying:bool=Montage_IsPlaying(self=@n5, Montage=/BRCosmetics/Animation/Game/MainPlayer/Emotes/Eastern_Bloc/Emote_EasternBloc_CMM_M)) -> (Output_Get:bool)
   n6 = Gravity Override State Function DEPRECATED(
    execute=@n1.then_0
    CharacterPartAnimInstance=Self-Reference()
    GameTimeInSeconds=GetGameTimeInSeconds()
    GravityOverrideParams=Get Gravity Override Params()
   )
   > Set GravityOverride(GravityOverride=@n6)
   NOT Boolean(A=@n4)
  ]}
]
variables [
 Gravity Override Params {type=GravityOverrideParamsStruct}
 GravityOverride {type=Vector}
 isCustomEmotePlaying {type=boolean}
 isTacticalSprintingAndJumping {type=boolean}
 NotCrouching {type=boolean, default=true}
]
