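# Physics asset for the M_LRG_CometHoliday character body that, by its name, supplies collision shapes for the cape.
# Capsules cover the pelvis, spine, clavicles, neck and back skirt bones, with a single sphere on the head.
# The spine_02 and spine_05 bodies stack several capsules, and the lower back skirt bones carry long ones.
# Arm, leg and foot bodies have no shapes, and no constraints join any bodies.
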PhysicsAsset /BRCosmetics/Characters/Player/Male/Large/Bodies/M_LRG_CometHoliday/Meshes/M_LRG_CometHoliday_CapeColliders
bodies {
 calf_l []
 calf_r []
 clavicle_l [{shapeType=Capsule, length=9.0192, radius=9.8871, center=<9.4266 -1.1806 -2.4977>, rotation=<rot 69.3297 1.4476 176.9462>}]
 clavicle_r [{shapeType=Capsule, length=9.0192, radius=9.8871, center=<-9.4266 1.1806 2.4977>, rotation=<rot 69.3297 1.4476 -3.0538>}]
 foot_l []
 foot_r []
 head [{shapeType=Sphere, radius=17.3125, center=<11.028 -0.174 0>}]
 lowerarm_l []
 lowerarm_r []
 neck_02 [{shapeType=Capsule, length=13.4375, radius=7.3175, center=<-0.2642 -0.3408 0>, rotation=<rot 90 -171.1193 -175.3978>}]
 pelvis [{shapeType=Capsule, length=13.687, radius=25.0454, center=<-1.7141 -8.4586 0>}]
 skirt_back_01_l [{shapeType=Capsule, length=27.0275, radius=18.498, center=<22.5339 -1.3107 -5.6949>, rotation=<rot -90 -6.7214 6.7216>}]
 skirt_back_01_r [{shapeType=Capsule, length=27.2533, radius=18.498, center=<21.6191 -1.2262 0>, rotation=<rot -90 -166.7373 -13.2625>}]
 skirt_back_02_l [{shapeType=Capsule, length=78.1395, radius=19.9883, center=<6.9819 -1.3108 -5.6949>, rotation=<rot 0 -179.9996 180>}]
 skirt_back_02_r [{shapeType=Capsule, length=78.3653, radius=19.9883, center=<6.0671 -1.2263 0>, rotation=<rot 0 -179.9996 0>}]
 spine_01 [{shapeType=Capsule, length=25.4614, radius=26.8297, center=<-4.4369 -4.5727 0>}]
 spine_02 [
  {shapeType=Capsule, length=20.9503, radius=25.664, center=<-1.711 -3.0427 0.0001>}
  {shapeType=Capsule, length=29.3956, radius=11.5807, center=<2.5024 0.4992 3.9825>, rotation=<rot 90 0.0049 0.0049>}
  {shapeType=Capsule, length=29.875, radius=11.1013, center=<2.4704 0.4991 -3.9825>, rotation=<rot 90 -0.0396 -0.0396>}
 ]
 spine_03 [{shapeType=Capsule, length=16.3049, radius=26.0831, center=<0.4139 -2.3352 0.0001>}]
 spine_04 [{shapeType=Capsule, length=18.1449, radius=22.8835, center=<-4.1771 -4.8638 0.0002>}]
 spine_05 [
  {shapeType=Capsule, length=0.2016, radius=24.9893, center=<-3.9211 -0.4455 0>}
  {shapeType=Capsule, length=13.9518, radius=13.9425, center=<-0.2916 0.4452 0>}
 ]
 thigh_l []
 thigh_r []
 upperarm_l []
 upperarm_r []
}
constraints []
